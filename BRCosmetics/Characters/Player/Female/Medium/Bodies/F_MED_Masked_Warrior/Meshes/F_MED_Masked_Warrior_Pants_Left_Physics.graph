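# Physics asset for the female Masked Warrior's left pant leg, modeling the loose fabric as jointed bone chains around the leg.
# Capsule bodies form five front-to-back chains beside pelvis, thigh and calf capsules, with narrow angular limits at chain roots and wider ones below.
# Front and middle chains attach directly to the thigh; both rear chains use small sphere base bodies.
# Free-angle cross constraints tie the mid-back chain to its neighbors.
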PhysicsAsset /BRCosmetics/Characters/Player/Female/Medium/Bodies/F_MED_Masked_Warrior/Meshes/F_MED_Masked_Warrior_Pants_Left_Physics
bodies {
 calf_l [{shapeType=Capsule, length=23.3478, radius=9.4529, center=<-11.123 -4.1332 -0.0921>, rotation=<rot 88.7225 -89.9992 90.0009>}]
 dyn_pant_leg_l_b_01 [{shapeType=Capsule, length=8, radius=4, shapeName=dyn_pant_leg_l_b_01_capsule, center=<0 0.5 1>, rotation=<rot 0 0 90>}]
 dyn_pant_leg_l_b_01_Base [{shapeType=Sphere, radius=0.505}]
 dyn_pant_leg_l_b_02 [{shapeType=Capsule, length=8, radius=4, shapeName=dyn_pant_leg_l_b_02_capsule, center=<0 0 1>, rotation=<rot 0 0 90>}]
 dyn_pant_leg_l_b_03 [{shapeType=Capsule, length=8, radius=4, shapeName=dyn_pant_leg_l_b_03_capsule, center=<0 0 2>, rotation=<rot 0 0 90>}]
 dyn_pant_leg_l_f_01 [{shapeType=Capsule, length=8, radius=4, shapeName=dyn_pant_leg_l_f_01_capsule, center=<0 0 2.9077>, rotation=<rot 0 0 90>}]
 dyn_pant_leg_l_f_02 [{shapeType=Capsule, length=11.9122, radius=5.343, shapeName=dyn_pant_leg_l_f_02_capsule, center=<0 0 1.5909>, rotation=<rot 0 0 90>}]
 dyn_pant_leg_l_f_03 [{shapeType=Capsule, length=11.9122, radius=5.343, shapeName=dyn_pant_leg_l_f_03_capsule, center=<0 0 0.5878>, rotation=<rot 0 0 90>}]
 dyn_pant_leg_l_m_01 [{shapeType=Capsule, length=8, radius=4, shapeName=dyn_pant_leg_l_m_01_capsule, center=<0 0 1.251>, rotation=<rot 0 0 90>}]
 dyn_pant_leg_l_m_02 [{shapeType=Capsule, length=12.8961, radius=4.419, shapeName=dyn_pant_leg_l_m_02_capsule, center=<0 0 2.4241>, rotation=<rot 0 0 -90>}]
 dyn_pant_leg_l_m_03 [{
   shapeType=Capsule
   length=11.5393
   radius=5.7759
   shapeName=dyn_pant_leg_l_m_03_capsule
   center=<0 0.1457 2.5891>
   rotation=<rot 0 0 -86.7799>
  }]
 dyn_pant_leg_l_m_04 [{
   shapeType=Capsule
   length=10.3359
   radius=5.8281
   shapeName=dyn_pant_leg_l_m_04_capsule
   center=<0 0.0105 0.4553>
   rotation=<rot 0 0 -88.6753>
  }]
 dyn_pant_leg_l_mb_01 [{shapeType=Capsule, length=8, radius=4, shapeName=dyn_pant_leg_l_mb_01_capsule, rotation=<rot 0 0 90>}]
 dyn_pant_leg_l_mb_01_Base [{shapeType=Sphere, radius=0.505}]
 dyn_pant_leg_l_mb_02 [{shapeType=Capsule, length=8, radius=4, shapeName=dyn_pant_leg_l_mb_02_capsule, center=<0 0 2.6816>, rotation=<rot 0 0 90>}]
 dyn_pant_leg_l_mb_03 [{shapeType=Capsule, length=8, radius=4, shapeName=dyn_pant_leg_l_mb_03_capsule, center=<0 0 2.4596>, rotation=<rot 0 0 90>}]
 dyn_pant_leg_l_mb_04 [{shapeType=Capsule, length=8, radius=4, shapeName=dyn_pant_leg_l_mb_04_capsule, center=<0 0 2>, rotation=<rot 0 0 90>}]
 dyn_pant_leg_l_mf_01 [{shapeType=Capsule, length=8, radius=4, shapeName=dyn_pant_leg_l_mf_01_capsule, center=<0 0 3.6513>, rotation=<rot 0 0 90>}]
 dyn_pant_leg_l_mf_02 [{shapeType=Capsule, length=11.4316, radius=5.516, shapeName=dyn_pant_leg_l_mf_02_capsule, center=<0 0 0.8667>, rotation=<rot 0 0 90>}]
 dyn_pant_leg_l_mf_03 [{
   shapeType=Capsule
   length=11.4316
   radius=5.516
   shapeName=dyn_pant_leg_l_mf_03_capsule
   center=<0.0068 -0.0776 0.9467>
   rotation=<rot 0.0005 5.0005 94.9999>
  }]
 pelvis [{shapeType=Capsule, length=13.5323, radius=6.8057, center=<-9.1835 -1.1958 0>}]
 thigh_l [{shapeType=Capsule, length=35.8127, radius=11.4263, center=<-24.3149 -0.63 1.5329>, rotation=<rot 89.4205 176.2432 175.5381>}]
}
constraints [
 {
  bone1Name=calf_l
  bone2Name=thigh_l
  swing1LimitDegrees=45
  swing1Motion=Free
  swing2LimitDegrees=45
  swing2Motion=Free
  twistLimitDegrees=15
  twistMotion=Free
 }
 {
  bone1Name=thigh_l
  bone2Name=pelvis
  swing1LimitDegrees=45
  swing1Motion=Free
  swing2LimitDegrees=45
  swing2Motion=Free
  twistLimitDegrees=45
  twistMotion=Free
 }
 {bone1Name=dyn_pant_leg_l_b_01, bone2Name=dyn_pant_leg_l_b_01_Base, swing1LimitDegrees=10, swing2LimitDegrees=10, twistLimitDegrees=5}
 {bone1Name=dyn_pant_leg_l_b_02, bone2Name=dyn_pant_leg_l_b_01, swing1LimitDegrees=20, swing2LimitDegrees=30, twistLimitDegrees=15}
 {bone1Name=dyn_pant_leg_l_b_03, bone2Name=dyn_pant_leg_l_b_02, swing1LimitDegrees=20, swing2LimitDegrees=30, twistLimitDegrees=15}
 {bone1Name=dyn_pant_leg_l_m_01, bone2Name=thigh_l, swing1LimitDegrees=5, swing2LimitDegrees=5, twistLimitDegrees=5}
 {bone1Name=dyn_pant_leg_l_m_02, bone2Name=dyn_pant_leg_l_m_01, swing1LimitDegrees=20, swing2LimitDegrees=30, twistLimitDegrees=15}
 {bone1Name=dyn_pant_leg_l_m_03, bone2Name=dyn_pant_leg_l_m_02, swing1LimitDegrees=20, swing2LimitDegrees=30, twistLimitDegrees=15}
 {bone1Name=dyn_pant_leg_l_m_04, bone2Name=dyn_pant_leg_l_m_03, swing1LimitDegrees=20, swing2LimitDegrees=30, twistLimitDegrees=15}
 {bone1Name=dyn_pant_leg_l_f_01, bone2Name=thigh_l, swing1LimitDegrees=5, swing2LimitDegrees=5, twistLimitDegrees=5}
 {bone1Name=dyn_pant_leg_l_f_02, bone2Name=dyn_pant_leg_l_f_01, swing1LimitDegrees=20, swing2LimitDegrees=30, twistLimitDegrees=15}
 {bone1Name=dyn_pant_leg_l_f_03, bone2Name=dyn_pant_leg_l_f_02, swing1LimitDegrees=20, swing2LimitDegrees=30, twistLimitDegrees=15}
 {bone1Name=dyn_pant_leg_l_mf_01, bone2Name=thigh_l, swing1LimitDegrees=5, swing2LimitDegrees=5, twistLimitDegrees=5}
 {bone1Name=dyn_pant_leg_l_mf_02, bone2Name=dyn_pant_leg_l_mf_01, swing1LimitDegrees=20, swing2LimitDegrees=30, twistLimitDegrees=15}
 {bone1Name=dyn_pant_leg_l_mf_03, bone2Name=dyn_pant_leg_l_mf_02, swing1LimitDegrees=20, swing2LimitDegrees=30, twistLimitDegrees=15}
 {bone1Name=dyn_pant_leg_l_mb_01, bone2Name=dyn_pant_leg_l_mb_01_Base, swing1LimitDegrees=5, swing2LimitDegrees=5, twistLimitDegrees=5}
 {bone1Name=dyn_pant_leg_l_mb_02, bone2Name=dyn_pant_leg_l_mb_01, swing1LimitDegrees=20, swing2LimitDegrees=30, twistLimitDegrees=15}
 {bone1Name=dyn_pant_leg_l_mb_03, bone2Name=dyn_pant_leg_l_mb_02, swing1LimitDegrees=20, swing2LimitDegrees=30, twistLimitDegrees=15}
 {bone1Name=dyn_pant_leg_l_mb_04, bone2Name=dyn_pant_leg_l_mb_03, swing1LimitDegrees=20, swing2LimitDegrees=30, twistLimitDegrees=15}
 {
  bone1Name=dyn_pant_leg_l_b_01_Base
  bone2Name=thigh_l
  swing1LimitDegrees=45
  swing1Motion=Free
  swing2LimitDegrees=45
  swing2Motion=Free
  twistLimitDegrees=45
  twistMotion=Free
 }
 {
  bone1Name=dyn_pant_leg_l_mb_01_Base
  bone2Name=thigh_l
  swing1LimitDegrees=45
  swing1Motion=Free
  swing2LimitDegrees=45
  swing2Motion=Free
  twistLimitDegrees=45
  twistMotion=Free
 }
 {
  bone1Name=dyn_pant_leg_l_mb_04
  bone2Name=dyn_pant_leg_l_m_04
  swing1LimitDegrees=45
  swing1Motion=Free
  swing2LimitDegrees=45
  swing2Motion=Free
  twistLimitDegrees=45
  twistMotion=Free
 }
 {
  bone1Name=dyn_pant_leg_l_mb_03
  bone2Name=dyn_pant_leg_l_b_03
  swing1LimitDegrees=45
  swing1Motion=Free
  swing2LimitDegrees=45
  swing2Motion=Free
  twistLimitDegrees=45
  twistMotion=Free
 }
]
