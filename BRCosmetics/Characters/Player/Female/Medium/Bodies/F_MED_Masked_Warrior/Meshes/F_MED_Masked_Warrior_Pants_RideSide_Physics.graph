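# Physics asset for the female medium Masked Warrior pants, RideSide variant, defining capsule bodies and constraints for dynamic cloth bones on the right leg only.
# Five bone chains (front, mid-front, middle, mid-back, back) hang from thigh_r with mostly twist-locked joints, alongside pelvis, thigh and calf capsules.
# The back root rotates freely, the front tip is angularly locked to its parent, and a free-rotation cross-constraint links the front and mid-front tips.
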PhysicsAsset /BRCosmetics/Characters/Player/Female/Medium/Bodies/F_MED_Masked_Warrior/Meshes/F_MED_Masked_Warrior_Pants_RideSide_Physics
bodies {
 calf_r [{shapeType=Capsule, length=26.5519, radius=9.2574, center=<13.3911 4.1331 0.0921>, rotation=<rot 88.7225 -89.9992 -89.9991>}]
 dyn_pant_leg_r_b_01 [{shapeType=Capsule, length=2.1181, radius=2.1075, shapeName=dyn_pant_leg_r_b_01_capsule, rotation=<rot 0 0 90>}]
 dyn_pant_leg_r_b_02 [{shapeType=Capsule, length=8.3312, radius=3.6688, shapeName=dyn_pant_leg_r_b_02_capsule, center=<0 0 -2>, rotation=<rot 0 0 60>}]
 dyn_pant_leg_r_b_03 [{shapeType=Capsule, length=6, radius=4, shapeName=dyn_pant_leg_r_b_03_capsule, center=<0 0 -2>, rotation=<rot 0 0 65>}]
 dyn_pant_leg_r_f_01 [{shapeType=Capsule, length=8, radius=4, shapeName=dyn_pant_leg_r_f_01_capsule, rotation=<rot 0 0 90>}]
 dyn_pant_leg_r_f_02 [{shapeType=Capsule, length=12.4394, radius=5.5606, shapeName=dyn_pant_leg_r_f_02_capsule, center=<0 0.3 -0.3>, rotation=<rot 0 4 90>}]
 dyn_pant_leg_r_f_03 [{shapeType=Capsule, length=7.4394, radius=7.0606, shapeName=dyn_pant_leg_r_f_03_capsule, center=<0 0 -3>, rotation=<rot 0 5 90>}]
 dyn_pant_leg_r_m_01 [{shapeType=Capsule, length=8, radius=4, shapeName=dyn_pant_leg_r_m_01_capsule, rotation=<rot 0 0 90>}]
 dyn_pant_leg_r_m_02 [{shapeType=Capsule, length=5.4935, radius=6.5065, shapeName=dyn_pant_leg_r_m_02_capsule, center=<0 0 -3>, rotation=<rot 0 2.5 90>}]
 dyn_pant_leg_r_m_03 [{
   shapeType=Capsule
   length=5.4935
   radius=6.5065
   shapeName=dyn_pant_leg_r_m_03_capsule
   center=<0 0 -4>
   rotation=<rot -0.0001 5.0009 89.9996>
  }]
 dyn_pant_leg_r_m_04 [{shapeType=Capsule, length=5.4394, radius=6.5606, shapeName=dyn_pant_leg_r_m_04_capsule, rotation=<rot 0 0 90>}]
 dyn_pant_leg_r_mb_01 [{shapeType=Capsule, length=8, radius=4, shapeName=dyn_pant_leg_r_mb_01_capsule, center=<0 0 -1.1643>, rotation=<rot 0 0 90>}]
 dyn_pant_leg_r_mb_02 [{
   shapeType=Capsule
   length=20.9394
   radius=4.5606
   shapeName=dyn_pant_leg_r_mb_02_capsule
   center=<-8.1756 0.8545 -3.8781>
   rotation=<rot -0.0001 -89.9985 -89.9987>
  }]
 dyn_pant_leg_r_mf_01 [{shapeType=Capsule, length=8, radius=4, shapeName=dyn_pant_leg_r_mf_01_capsule, rotation=<rot 0 0 90>}]
 dyn_pant_leg_r_mf_02 [{shapeType=Capsule, length=7.4394, radius=4.5606, shapeName=dyn_pant_leg_r_mf_02_capsule, rotation=<rot 0.0001 5.0001 90>}]
 dyn_pant_leg_r_mf_03 [{shapeType=Capsule, length=6.4394, radius=5.5606, shapeName=dyn_pant_leg_r_mf_03_capsule, center=<-1.5 0 -4>, rotation=<rot 10 15 90>}]
 pelvis [{shapeType=Capsule, length=13.5323, radius=6.8057, center=<-9.1835 -1.1958 0>}]
 thigh_r [{shapeType=Capsule, length=22.827, radius=9.9493, shapeName=thigh_r_capsule, center=<24.0267 0 0>, rotation=<rot 90 0 0>}]
}
constraints [
 {bone1Name=dyn_pant_leg_r_m_01, bone2Name=thigh_r, swing1LimitDegrees=5, swing2LimitDegrees=5, twistLimitDegrees=5}
 {bone1Name=dyn_pant_leg_r_f_01, bone2Name=thigh_r, swing1LimitDegrees=5, swing2LimitDegrees=5, twistLimitDegrees=5}
 {bone1Name=dyn_pant_leg_r_mf_01, bone2Name=thigh_r, swing1LimitDegrees=10, swing2LimitDegrees=20, twistLimitDegrees=5, twistMotion=Locked}
 {
  bone1Name=dyn_pant_leg_r_b_01
  bone2Name=thigh_r
  swing1LimitDegrees=45
  swing1Motion=Free
  swing2LimitDegrees=45
  swing2Motion=Free
  twistLimitDegrees=45
  twistMotion=Free
 }
 {
  bone1Name=dyn_pant_leg_r_b_02
  bone2Name=dyn_pant_leg_r_b_01
  swing1LimitDegrees=20
  swing2LimitDegrees=30
  twistLimitDegrees=45
  twistMotion=Locked
 }
 {
  bone1Name=dyn_pant_leg_r_b_03
  bone2Name=dyn_pant_leg_r_b_02
  swing1LimitDegrees=10
  swing2LimitDegrees=30
  twistLimitDegrees=45
  twistMotion=Locked
 }
 {
  bone1Name=dyn_pant_leg_r_f_02
  bone2Name=dyn_pant_leg_r_f_01
  swing1LimitDegrees=5
  swing2LimitDegrees=30
  twistLimitDegrees=45
  twistMotion=Locked
 }
 {
  bone1Name=dyn_pant_leg_r_f_03
  bone2Name=dyn_pant_leg_r_f_02
  swing1LimitDegrees=10
  swing1Motion=Locked
  swing2LimitDegrees=30
  swing2Motion=Locked
  twistLimitDegrees=45
  twistMotion=Locked
 }
 {
  bone1Name=dyn_pant_leg_r_m_02
  bone2Name=dyn_pant_leg_r_m_01
  swing1LimitDegrees=0
  swing2LimitDegrees=30
  twistLimitDegrees=45
  twistMotion=Locked
 }
 {
  bone1Name=dyn_pant_leg_r_m_03
  bone2Name=dyn_pant_leg_r_m_02
  swing1LimitDegrees=20
  swing2LimitDegrees=30
  twistLimitDegrees=45
  twistMotion=Locked
 }
 {bone1Name=dyn_pant_leg_r_mb_01, bone2Name=thigh_r, swing1LimitDegrees=10, swing2LimitDegrees=20, twistLimitDegrees=45, twistMotion=Locked}
 {
  bone1Name=dyn_pant_leg_r_mf_02
  bone2Name=dyn_pant_leg_r_mf_01
  swing1LimitDegrees=20
  swing2LimitDegrees=30
  twistLimitDegrees=45
  twistMotion=Locked
 }
 {
  bone1Name=dyn_pant_leg_r_m_04
  bone2Name=dyn_pant_leg_r_m_03
  swing1LimitDegrees=10
  swing2LimitDegrees=30
  twistLimitDegrees=45
  twistMotion=Locked
 }
 {
  bone1Name=dyn_pant_leg_r_mb_02
  bone2Name=dyn_pant_leg_r_mb_01
  swing1LimitDegrees=10
  swing2LimitDegrees=30
  twistLimitDegrees=45
  twistMotion=Locked
 }
 {
  bone1Name=dyn_pant_leg_r_mf_03
  bone2Name=dyn_pant_leg_r_mf_02
  swing1LimitDegrees=10
  swing2LimitDegrees=30
  twistLimitDegrees=45
  twistMotion=Locked
 }
 {
  bone1Name=dyn_pant_leg_r_f_03
  bone2Name=dyn_pant_leg_r_mf_03
  swing1LimitDegrees=45
  swing1Motion=Free
  swing2LimitDegrees=45
  swing2Motion=Free
  twistLimitDegrees=45
  twistMotion=Free
 }
]
