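# Physics asset for the scarf on the female medium Masked Warrior player cosmetic body, defining collision bodies and joints for its dynamic scarf bones.
# Left and right two-bone chains use box bodies and hang from a spine_04 capsule, each bone constrained to its parent.
# Cross constraints also tie the left and right chains together at both levels, forming a ladder, with all joints sharing the same swing and twist limits.
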PhysicsAsset /BRCosmetics/Characters/Player/Female/Medium/Bodies/F_MED_Masked_Warrior/Meshes/F_MED_Masked_Warrior_Scarf_Physics
bodies {
 dyn_scarf_l_01 [{shapeType=Box, extentX=5.2727, extentY=3.1713, extentZ=2.4476, center=<2.7347 -0.1329 -0.0145>, rotation=<rot 0 0 -26.0751>}]
 dyn_scarf_l_02 [{shapeType=Box, extentX=4.8125, extentY=2.6284, extentZ=2.7509, center=<1.418 -0.3842 -0.4683>, rotation=<rot 0 0 -26.0751>}]
 dyn_scarf_r_01 [{shapeType=Box, extentX=4.0289, extentY=3.3977, extentZ=1.5941, center=<2.0598 -0.6907 0.4415>}]
 dyn_scarf_r_02 [{shapeType=Box, extentX=6.8573, extentY=4.6032, extentZ=3.0529, center=<0.9492 -1.2104 -0.0385>}]
 spine_04 [{shapeType=Capsule, length=1.1834, radius=13.4656}]
}
constraints [
 {bone1Name=dyn_scarf_l_01, bone2Name=spine_04, swing1LimitDegrees=45, swing2LimitDegrees=45, twistLimitDegrees=15}
 {bone1Name=dyn_scarf_l_02, bone2Name=dyn_scarf_l_01, swing1LimitDegrees=45, swing2LimitDegrees=45, twistLimitDegrees=15}
 {bone1Name=dyn_scarf_r_01, bone2Name=spine_04, swing1LimitDegrees=45, swing2LimitDegrees=45, twistLimitDegrees=15}
 {bone1Name=dyn_scarf_r_02, bone2Name=dyn_scarf_r_01, swing1LimitDegrees=45, swing2LimitDegrees=45, twistLimitDegrees=15}
 {bone1Name=dyn_scarf_l_01, bone2Name=dyn_scarf_r_01, swing1LimitDegrees=45, swing2LimitDegrees=45, twistLimitDegrees=15}
 {bone1Name=dyn_scarf_l_02, bone2Name=dyn_scarf_r_02, swing1LimitDegrees=45, swing2LimitDegrees=45, twistLimitDegrees=15}
]
