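# Post-animation AnimBlueprint for the female Masked Warrior cosmetic body, correcting and simulating the pant-leg dynamic bones.
# Branches from cached pose CRCache_01 layer additive bone offsets, Look At and rotation constraints toward thighs and calves, with some branches blended by bool.
# RigidBody nodes with per-side physics assets, one pair taking skydiving force, and gravity-driven Anim Dynamics add secondary motion.
# Speed, skydive, jump, crouch, knee-slide and leg-bend variables drive node alphas.
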AnimBlueprint /BRCosmetics/Characters/Player/Female/Medium/Bodies/F_MED_Masked_Warrior/Meshes/F_MED_MMasked_Warrior_Body_Skeleton_AnimBP
target_skeleton=&/F_MED_Masked_Warrior_Body_Skeleton
graphs [
 AnimGraph {nodes [
   n1 = "Transform (Modify) Bone - Bone: dyn_pant_leg_l_mb_01_Base"!(ComponentPose, Alpha=1, Translation="0,0,1.5", Rotation="10,0,0") -> (Pose) {
    rotationMode=BMM_Additive
    rotationSpace=BCS_BoneSpace
    translationMode=BMM_Additive
    translationSpace=BCS_BoneSpace
    alphaBoolBlend {blendOption=Linear}
    alphaScaleBiasClamp {bInterpResult=true}
    boneToModify {boneName=dyn_pant_leg_l_mb_01_Base}
    rotation {pitch=10}
    translation {z=1.5}
   }
   "Transform (Modify) Bone - Bone: dyn_pant_leg_l_b_01"!(ComponentPose=@n1, Alpha=1, Translation="0,0,2", Rotation="25,0,0.002") -> (Pose) {
    rotationMode=BMM_Additive
    rotationSpace=BCS_BoneSpace
    translationMode=BMM_Additive
    translationSpace=BCS_BoneSpace
    alphaBoolBlend {blendOption=Linear}
    alphaScaleBiasClamp {bInterpResult=true}
    boneToModify {boneName=dyn_pant_leg_l_b_01}
    rotation {pitch=25, roll=0.002}
    translation {z=2}
   }
   n2 = Get PawnSpeed2D()
   n3 = "Transform (Modify) Bone - Bone: dyn_pant_leg_r_mb_01"(ComponentPose=Local To Component(LocalPose=Use cached pose 'CRCache_01'()), Alpha=@n2, Translation="5,-5,0", Rotation="0,0,0", Scale="1,1,1") {
    translationMode=BMM_Additive
    alphaBoolBlend {blendOption=Linear}
    alphaScaleBiasClamp {bMapRange=true, inRange {max=500}}
    boneToModify {boneName=dyn_pant_leg_r_mb_01}
    translation=<5 -5>
   }
   n4 = "Transform (Modify) Bone - Bone: dyn_pant_leg_l_mb_01"(ComponentPose=@n3, Alpha=@n2, Translation="5,-5,0", Rotation="0,0,0", Scale="1,1,1") {
    translationMode=BMM_Additive
    alphaBoolBlend {blendOption=Linear}
    alphaScaleBiasClamp {bMapRange=true, inRange {max=500}}
    boneToModify {boneName=dyn_pant_leg_l_mb_01}
    translation=<5 -5>
   }
   n5 = Break Fort Anim Input Skydiving External Force(FortAnimInput_SkydivingExternalForce=Get SkydivingExternalForceInput()) -> (
    bUseSkydivingVectorForce=~
    SkydivingVectorForce=~
    HeadToPelvisDirection=~
    FloatingMultiplier=~
    FloatingAdditive=~
    DivingMultiplier=~
    DivingAdditive=~
    ParachutingMultiplier=~
    ParachutingAdditive=~
   )
   n6 = Get SkydiveAlpha() -> (SkydiveAlpha:double)
   n7 = RigidBody(ComponentPose=@n4, Alpha=@n6, ExternalForce=@n5.SkydivingVectorForce) {
    bClampLinearTranslationLimitToRefPose=true
    bOverrideWorldGravity=true
    overridePhysicsAsset=&/F_MED_Masked_Warrior_Pants_Left_Physics
    simulationSpace=BaseBoneSpace
    alphaBoolBlend {blendOption=Linear}
    alphaScaleBiasClamp {bInterpResult=true}
    baseBoneRef {boneName=root}
    componentAppliedLinearAccClamp=<9000 9000 9000>
    componentLinearAccScale=<0.8 0.8 0.8>
    componentLinearVelScale=<0.4 0.4 0.4>
    overrideWorldGravity {z=-980}
   }
   n8 = RigidBody(ComponentPose=@n7, Alpha=@n6, ExternalForce=@n5.SkydivingVectorForce) {
    bClampLinearTranslationLimitToRefPose=true
    bOverrideWorldGravity=true
    overridePhysicsAsset=&/F_MED_Masked_Warrior_Pants_RideSide_Physics
    simulationSpace=BaseBoneSpace
    alphaBoolBlend {blendInTime=0.1, blendOption=Linear, blendOutTime=0.1}
    alphaScaleBiasClamp {bInterpResult=true}
    baseBoneRef {boneName=root}
    componentAppliedLinearAccClamp=<9000 9000 9000>
    componentLinearAccScale=<0.15 0.15 0.15>
    componentLinearVelScale=<0.8 0.8 0.8>
    overrideWorldGravity {z=-980}
   }
   n9 = Get Jumping() -> (Jumping:double)
   n10 = "Transform (Modify) Bone - Bone: dyn_pant_leg_r_mb_01"(ComponentPose=@n8, Alpha=@n9, Translation="-5,-10,0", Rotation="90,0,0", Scale="1,1,1") {
    translationMode=BMM_Additive
    alphaBoolBlend {blendInTime=0.1, blendOption=Linear, blendOutTime=0.1}
    alphaScaleBiasClamp {bInterpResult=true}
    boneToModify {boneName=dyn_pant_leg_r_mb_01}
    rotation {pitch=90}
    translation=<-5 -10>
   }
   n11 = "Transform (Modify) Bone - Bone: dyn_pant_leg_r_b_02"(ComponentPose=@n10, Alpha=@n9, Translation="-5,-5,0", Rotation="10,0,0", Scale="1,1,1") {
    translationMode=BMM_Additive
    alphaBoolBlend {blendInTime=0.1, blendOption=Linear, blendOutTime=0.1}
    alphaScaleBiasClamp {bInterpResult=true}
    boneToModify {boneName=dyn_pant_leg_r_b_02}
    rotation {pitch=10}
    translation=<-5 -5>
   }
   n12 = "Transform (Modify) Bone - Bone: dyn_pant_leg_r_m_01"(ComponentPose=@n11, Alpha=@n9, Translation="0,0,-5", Rotation="10,0,0", Scale="1,1,1") {
    rotationMode=BMM_Additive
    rotationSpace=BCS_BoneSpace
    translationMode=BMM_Additive
    translationSpace=BCS_BoneSpace
    alphaBoolBlend {blendOption=Linear}
    alphaScaleBiasClamp {bInterpResult=true}
    boneToModify {boneName=dyn_pant_leg_r_m_01}
    rotation {pitch=10}
    translation {z=-5}
   }
   n13 = Anim Dynamics - Bone: dyn_pant_leg_r_mf_03(ComponentPose=@n12, Alpha=@n9, ExternalForce="0,0,0") {
    bGravityOverrideInSimSpace=true
    bUseGravityOverride=true
    simulationSpace=RootRelative
    alphaBoolBlend {blendInTime=0.1, blendOption=Linear, blendOutTime=0.1}
    alphaScaleBiasClamp {bInterpResult=true}
    boundBone {boneName=dyn_pant_leg_r_mf_03}
    componentAppliedLinearAccClamp=<10000 10000>
    componentLinearAccScale=<0.5 0.5 0.5>
    gravityOverride {z=-980}
    physicsBodyDefinitions [{
      sphereCollisionRadius=0
      boundBone {boneName=dyn_pant_leg_r_mf_03}
      boxExtents=<5 5 2.5>
      constraintSetup {angularLimitsMax=<0 2 20>, angularLimitsMin=<0 -2 0>, angularTarget=<1 0 0>}
      localJointOffset=<0.1 0 0>
     }]
    relativeSpaceBone {boneName=root}
   }
   n14 = Anim Dynamics - Bone: dyn_pant_leg_r_mf_04(ComponentPose=@n13, Alpha=@n9, ExternalForce="0,0,0") {
    bGravityOverrideInSimSpace=true
    bUseGravityOverride=true
    simulationSpace=RootRelative
    alphaBoolBlend {blendInTime=0.1, blendOption=Linear, blendOutTime=0.1}
    alphaScaleBiasClamp {bInterpResult=true}
    boundBone {boneName=dyn_pant_leg_r_mf_04}
    componentAppliedLinearAccClamp=<10000 10000>
    componentLinearAccScale=<0.5 0.5 0.5>
    gravityOverride {z=-980}
    physicsBodyDefinitions [{
      sphereCollisionRadius=0
      boundBone {boneName=dyn_pant_leg_r_mf_04}
      boxExtents=<5 5 2.5>
      constraintSetup {angularLimitsMax=<0 2 20>, angularLimitsMin=<0 -2 0>, angularTarget=<1 0 0>}
      localJointOffset=<0.1 0 0>
     }]
    relativeSpaceBone {boneName=root}
   }
   n15 = Anim Dynamics - Bone: dyn_pant_leg_r_mb_03(ComponentPose=@n14, Alpha=@n9, ExternalForce="0,0,0") {
    bGravityOverrideInSimSpace=true
    bUseGravityOverride=true
    simulationSpace=RootRelative
    alphaBoolBlend {blendInTime=0.1, blendOption=Linear, blendOutTime=0.1}
    alphaScaleBiasClamp {bInterpResult=true}
    boundBone {boneName=dyn_pant_leg_r_mb_03}
    componentAppliedLinearAccClamp=<10000 10000>
    componentLinearAccScale=<0.5 0.5 0.5>
    gravityOverride {z=-980}
    physicsBodyDefinitions [{
      sphereCollisionRadius=0
      boundBone {boneName=dyn_pant_leg_r_mb_03}
      boxExtents=<5 5 2.5>
      constraintSetup {angularLimitsMax=<0 2 20>, angularLimitsMin=<0 -2 -20>, angularTarget=<1 0 0>}
      localJointOffset=<0.1 0 0>
     }]
    relativeSpaceBone {boneName=root}
   }
   n16 = Anim Dynamics - Bone: dyn_pant_leg_r_mb_04(ComponentPose=@n15, Alpha=@n9, ExternalForce="0,0,0") {
    bGravityOverrideInSimSpace=true
    bUseGravityOverride=true
    simulationSpace=RootRelative
    alphaBoolBlend {blendInTime=0.1, blendOption=Linear, blendOutTime=0.1}
    alphaScaleBiasClamp {bInterpResult=true}
    boundBone {boneName=dyn_pant_leg_r_mb_04}
    componentAppliedLinearAccClamp=<10000 10000>
    componentLinearAccScale=<0.5 0.5 0.5>
    gravityOverride {z=-980}
    physicsBodyDefinitions [{
      sphereCollisionRadius=0
      boundBone {boneName=dyn_pant_leg_r_mb_04}
      boxExtents=<5 5 2.5>
      constraintSetup {angularLimitsMax=<0 2 20>, angularLimitsMin=<0 -2 -20>, angularTarget=<1 0 0>}
      localJointOffset=<0.1 0 0>
     }]
    relativeSpaceBone {boneName=root}
   }
   n17 = Anim Dynamics - Bone: dyn_pant_leg_r_m_03(ComponentPose=@n16, Alpha=@n9, ExternalForce="0,0,0") {
    bGravityOverrideInSimSpace=true
    bUseGravityOverride=true
    simulationSpace=RootRelative
    alphaBoolBlend {blendInTime=0.1, blendOption=Linear, blendOutTime=0.1}
    alphaScaleBiasClamp {bInterpResult=true}
    boundBone {boneName=dyn_pant_leg_r_m_03}
    componentAppliedLinearAccClamp=<10000 10000>
    componentLinearAccScale=<0.5 0.5 0.5>
    gravityOverride {z=-980}
    physicsBodyDefinitions [{
      sphereCollisionRadius=0
      boundBone {boneName=dyn_pant_leg_r_m_03}
      boxExtents=<5 5 2.5>
      constraintSetup {angularLimitsMax=<0 2 20>, angularLimitsMin=<0 -2 -20>, angularTarget=<1 0 0>}
      localJointOffset=<0.1 0 0>
     }]
    relativeSpaceBone {boneName=root}
   }
   n18 = Anim Dynamics - Bone: dyn_pant_leg_r_m_04(ComponentPose=@n17, Alpha=@n9, ExternalForce="0,0,0") {
    bGravityOverrideInSimSpace=true
    bUseGravityOverride=true
    simulationSpace=RootRelative
    alphaBoolBlend {blendInTime=0.1, blendOption=Linear, blendOutTime=0.1}
    alphaScaleBiasClamp {bInterpResult=true}
    boundBone {boneName=dyn_pant_leg_r_m_04}
    componentAppliedLinearAccClamp=<10000 10000>
    componentLinearAccScale=<0.5 0.5 0.5>
    gravityOverride {z=-980}
    physicsBodyDefinitions [{
      sphereCollisionRadius=0
      boundBone {boneName=dyn_pant_leg_r_m_04}
      boxExtents=<5 5 2.5>
      constraintSetup {angularLimitsMax=<0 2 20>, angularLimitsMin=<0 -2 -20>, angularTarget=<1 0 0>}
      localJointOffset=<0.1 0 0>
     }]
    relativeSpaceBone {boneName=root}
   }
   n19 = "Transform (Modify) Bone - Bone: dyn_pant_leg_r_f_01"!(ComponentPose=@n18, bAlphaBoolEnabled:bool=True, Translation="0,0,0", Rotation="12,0,0") -> (Pose) {
    alphaInputType=Bool
    rotationMode=BMM_Additive
    rotationSpace=BCS_BoneSpace
    translationMode=BMM_Additive
    translationSpace=BCS_BoneSpace
    alphaBoolBlend {blendOption=Linear}
    boneToModify {boneName=dyn_pant_leg_r_f_01}
    rotation {pitch=12}
   }
   n20 = "Transform (Modify) Bone - Bone: dyn_pant_leg_r_m_01"!(ComponentPose=@n19, bAlphaBoolEnabled:bool=True, Translation="0,0,-0.5", Rotation="5,0,0") -> (Pose) {
    alphaInputType=Bool
    rotationMode=BMM_Additive
    rotationSpace=BCS_BoneSpace
    translationMode=BMM_Additive
    translationSpace=BCS_BoneSpace
    alphaBoolBlend {blendOption=Linear}
    boneToModify {boneName=dyn_pant_leg_r_m_01}
    rotation {pitch=5}
    translation {z=-0.5}
   }
   n21 = "Transform (Modify) Bone - Bone: dyn_pant_leg_r_mb_01"!(ComponentPose=@n20, bAlphaBoolEnabled:bool=True, Translation="0,0,-5", Rotation="10,0,0") -> (Pose) {
    alphaInputType=Bool
    rotationMode=BMM_Additive
    rotationSpace=BCS_BoneSpace
    translationMode=BMM_Additive
    translationSpace=BCS_BoneSpace
    alphaBoolBlend {blendOption=Linear}
    boneToModify {boneName=dyn_pant_leg_r_mb_01}
    rotation {pitch=10}
    translation {z=-5}
   }
   n22 = "Transform (Modify) Bone - Bone: dyn_pant_leg_r_m_02"!(ComponentPose=@n21, bAlphaBoolEnabled:bool=True, Translation="0,0,0", Rotation="-5,0,0") -> (Pose) {
    alphaInputType=Bool
    rotationMode=BMM_Additive
    rotationSpace=BCS_BoneSpace
    translationMode=BMM_Additive
    translationSpace=BCS_BoneSpace
    alphaBoolBlend {blendOption=Linear}
    boneToModify {boneName=dyn_pant_leg_r_m_02}
    rotation {pitch=-5}
   }
   n23 = "Transform (Modify) Bone - Bone: dyn_pant_leg_r_mf_01"!(ComponentPose=@n22, bAlphaBoolEnabled:bool=True, Translation="0,0,-2", Rotation="5,0,0") -> (Pose) {
    alphaInputType=Bool
    rotationMode=BMM_Additive
    rotationSpace=BCS_BoneSpace
    translationMode=BMM_Additive
    translationSpace=BCS_BoneSpace
    alphaBoolBlend {blendOption=Linear}
    boneToModify {boneName=dyn_pant_leg_r_mf_01}
    rotation {pitch=5}
    translation {z=-2}
   }
   n24 = "Transform (Modify) Bone - Bone: dyn_pant_leg_l_f_01"!(ComponentPose=@n23, bAlphaBoolEnabled:bool=True, Translation="0,0,0", Rotation="10,0,0") -> (Pose) {
    alphaInputType=Bool
    rotationMode=BMM_Additive
    rotationSpace=BCS_BoneSpace
    translationMode=BMM_Additive
    translationSpace=BCS_BoneSpace
    alphaBoolBlend {blendOption=Linear}
    boneToModify {boneName=dyn_pant_leg_l_f_01}
    rotation {pitch=10}
   }
   n25 = "Transform (Modify) Bone - Bone: dyn_pant_leg_l_mf_01"!(ComponentPose=@n24, bAlphaBoolEnabled:bool=True, Translation="0,0,0", Rotation="10,0,0") -> (Pose) {
    alphaInputType=Bool
    rotationMode=BMM_Additive
    rotationSpace=BCS_BoneSpace
    translationMode=BMM_Additive
    translationSpace=BCS_BoneSpace
    alphaBoolBlend {blendOption=Linear}
    boneToModify {boneName=dyn_pant_leg_l_mf_01}
    rotation {pitch=10}
   }
   n26 = "Transform (Modify) Bone - Bone: dyn_pant_leg_r_b_01"!(ComponentPose=@n25, bAlphaBoolEnabled:bool=True, Translation="0,0,0", Rotation="10,0,0", Scale="1,1,1") -> (Pose) {
    alphaInputType=Bool
    rotationMode=BMM_Additive
    rotationSpace=BCS_BoneSpace
    translationMode=BMM_Additive
    translationSpace=BCS_BoneSpace
    alphaBoolBlend {blendInTime=0.2, blendOption=Linear, blendOutTime=1}
    boneToModify {boneName=dyn_pant_leg_r_b_01}
    rotation {pitch=10}
   }
   n27 = "Transform (Modify) Bone - Bone: dyn_pant_leg_r_mb_01"!(ComponentPose=@n26, bAlphaBoolEnabled=True, Translation="0,0,3", Rotation="-20,5,0", Scale:Vector="1,1,1") -> (Pose) {
    alphaInputType=Bool
    rotationMode=BMM_Additive
    rotationSpace=BCS_BoneSpace
    translationMode=BMM_Additive
    translationSpace=BCS_BoneSpace
    alphaBoolBlend {blendInTime=0.2, blendOption=Linear, blendOutTime=1}
    boneToModify {boneName=dyn_pant_leg_r_mb_01}
    rotation {pitch=-20, yaw=5}
    translation {z=3}
   }
   n28 = "Transform (Modify) Bone - Bone: dyn_pant_leg_r_mf_01"!(ComponentPose=@n27, bAlphaBoolEnabled=True, Translation="0,0,2.55", Rotation="0,5,0", Scale:Vector="1,1,1") -> (Pose) {
    alphaInputType=Bool
    rotationMode=BMM_Additive
    rotationSpace=BCS_BoneSpace
    translationMode=BMM_Additive
    translationSpace=BCS_BoneSpace
    alphaBoolBlend {blendInTime=0.2, blendOption=Linear, blendOutTime=1}
    boneToModify {boneName=dyn_pant_leg_r_mf_01}
    rotation {yaw=5}
    translation {z=2.55}
   }
   n29 = Get SkydiveAlpha()
   n30 = "Transform (Modify) Bone - Bone: dyn_pant_leg_r_mb_01_Base"(ComponentPose=@n28, Alpha=@n29, Translation="0,0,0", Rotation="10.5,-15,-6", Scale="1,1,1") {
    rotationMode=BMM_Additive
    rotationSpace=BCS_BoneSpace
    translationMode=BMM_Additive
    translationSpace=BCS_BoneSpace
    alphaBoolBlend {blendOption=Linear}
    alphaScaleBiasClamp {bInterpResult=true}
    boneToModify {boneName=dyn_pant_leg_r_mb_01_Base}
    rotation=<rot 10.5 -15 -6>
   }
   n31 = "Transform (Modify) Bone - Bone: dyn_pant_leg_l_mb_01_Base"!(ComponentPose=@n30, Alpha=@n29, Translation="0,0,0", Rotation="25,-15,-10", Scale:Vector="1,1,1") -> (Pose) {
    rotationMode=BMM_Additive
    rotationSpace=BCS_BoneSpace
    translationMode=BMM_Additive
    translationSpace=BCS_BoneSpace
    alphaBoolBlend {blendOption=Linear}
    alphaScaleBiasClamp {bInterpResult=true}
    boneToModify {boneName=dyn_pant_leg_l_mb_01_Base}
    rotation=<rot 25 -15 -10>
   }
   n32 = "Transform (Modify) Bone - Bone: dyn_pant_leg_l_m_03"(ComponentPose=@n31, Alpha=@n29, Translation="0,0,0", Rotation="25,0,0", Scale="1,1,1") {
    rotationMode=BMM_Additive
    rotationSpace=BCS_BoneSpace
    translationMode=BMM_Additive
    translationSpace=BCS_BoneSpace
    alphaBoolBlend {blendOption=Linear}
    alphaScaleBiasClamp {bInterpResult=true}
    boneToModify {boneName=dyn_pant_leg_l_m_03}
    rotation {pitch=25}
   }
   n33 = Get FPS Compensation Alpha() -> (FPS Compensation Alpha:double)
   n34 = Get RightLegForwardBendCompensationAlpha() -> (RightLegForwardBendCompensationAlpha:double)
   n35 = Get PawnSpeed2D()
   n36 = "Transform (Modify) Bone - Bone: dyn_pant_leg_r_mb_01"!(
    ComponentPose=Local To Component(LocalPose=Use cached pose 'CRCache_01'())
    Alpha:float=@n35
    Translation="5,-5,0"
    Rotation="0,0,0"
    Scale="1,1,1"
   ) -> (Pose) {
    translationMode=BMM_Additive
    alphaBoolBlend {blendOption=Linear}
    alphaScaleBiasClamp {bMapRange=true, inRange {max=500}}
    boneToModify {boneName=dyn_pant_leg_r_mb_01}
    translation=<5 -5>
   }
   n37 = "Transform (Modify) Bone - Bone: dyn_pant_leg_l_mb_01"(ComponentPose=@n36, Alpha=@n35, Translation="5,-5,0", Rotation="0,0,0", Scale="1,1,1") {
    translationMode=BMM_Additive
    alphaBoolBlend {blendOption=Linear}
    alphaScaleBiasClamp {bMapRange=true, inRange {max=500}}
    boneToModify {boneName=dyn_pant_leg_l_mb_01}
    translation=<5 -5>
   }
   n38 = Get SkydiveAlpha()
   n39 = RigidBody(ComponentPose=@n37, Alpha=@n38) {
    bClampLinearTranslationLimitToRefPose=true
    bOverrideWorldGravity=true
    overridePhysicsAsset=&F_MED_Masked_Warrior_Pants_Left_Physics
    simulationSpace=BaseBoneSpace
    alphaBoolBlend {blendOption=Linear}
    alphaScaleBiasClamp {bInterpResult=true}
    baseBoneRef {boneName=root}
    componentAppliedLinearAccClamp=<9500 9500 9500>
    componentLinearAccScale=<0.8 0.8 0.8>
    componentLinearVelScale=<0.4 0.4 0.4>
    overrideWorldGravity {z=-980}
   }
   n40 = RigidBody(ComponentPose=@n39, Alpha=@n38) {
    bClampLinearTranslationLimitToRefPose=true
    bOverrideWorldGravity=true
    overridePhysicsAsset=&F_MED_Masked_Warrior_Pants_RideSide_Physics
    simulationSpace=BaseBoneSpace
    alphaBoolBlend {blendInTime=0.1, blendOption=Linear, blendOutTime=0.1}
    alphaScaleBiasClamp {bInterpResult=true}
    baseBoneRef {boneName=root}
    componentAppliedLinearAccClamp=<9500 9500 9500>
    componentLinearAccScale=<0.15 0.15 0.15>
    componentLinearVelScale=<0.8 0.8 0.8>
    overrideWorldGravity {z=-980}
   }
   n41 = Look At - Bone: dyn_pant_leg_l_f_01(ComponentPose=@n40, Alpha=@n34) {
    alphaBoolBlend {blendOption=Linear}
    alphaScaleBiasClamp {bClampResult=true, bInterpResult=true, clampMax=0.8, interpSpeedDecreasing=5, interpSpeedIncreasing=7}
    boneToModify {boneName=dyn_pant_leg_l_f_01}
    lookAtLocation=<-29.9278 13.1293 0>
    lookAtTarget {boneReference {boneName=thigh_l}}
    lookAt_Axis {axis=<1 0>}
   }
   n42 = Get LeftLegForwardBendCompensationAlpha() -> (LeftLegForwardBendCompensationAlpha:double)
   n43 = Look At - Bone: dyn_pant_leg_r_f_01(ComponentPose=@n41, Alpha=@n42) {
    alphaBoolBlend {blendOption=Linear}
    alphaScaleBiasClamp {bClampResult=true, bInterpResult=true, clampMax=0.8, interpSpeedDecreasing=5, interpSpeedIncreasing=7}
    boneToModify {boneName=dyn_pant_leg_r_f_01}
    lookAtLocation=<-13.0766 6.229 7.8309>
    lookAtTarget {boneReference {boneName=thigh_r}}
    lookAt_Axis {axis=<1 0>}
   }
   n44 = Look At - Bone: dyn_pant_leg_l_mf_01(ComponentPose=@n43, Alpha=@n34) {
    alphaBoolBlend {blendOption=Linear}
    alphaScaleBiasClamp {bClampResult=true, bInterpResult=true, clampMax=0.8, interpSpeedDecreasing=5, interpSpeedIncreasing=7}
    boneToModify {boneName=dyn_pant_leg_l_mf_01}
    lookAtLocation=<-37.6179 8.1616 -7.2206>
    lookAtTarget {boneReference {boneName=thigh_l}}
    lookAt_Axis {axis=<1 0>}
   }
   n45 = Look At - Bone: dyn_pant_leg_r_mf_01(ComponentPose=@n44, Alpha=@n42) {
    alphaBoolBlend {blendOption=Linear}
    alphaScaleBiasClamp {bClampResult=true, bInterpResult=true, clampMax=0.8, interpSpeedDecreasing=5, interpSpeedIncreasing=7}
    boneToModify {boneName=dyn_pant_leg_r_mf_01}
    lookAtLocation=<-14.6627 -1.1471 3.1567>
    lookAtTarget {boneReference {boneName=thigh_r}}
    lookAt_Axis {axis=<1 0>}
   }
   n46 = "Transform (Modify) Bone - Bone: dyn_pant_leg_r_f_01"!(ComponentPose=@n45, Alpha:float=@n33, Rotation="5,0,0") -> (Pose) {
    rotationMode=BMM_Additive
    rotationSpace=BCS_BoneSpace
    translationMode=BMM_Additive
    translationSpace=BCS_BoneSpace
    alphaBoolBlend {blendOption=Linear}
    alphaScaleBiasClamp {bInterpResult=true, interpSpeedDecreasing=5, interpSpeedIncreasing=7, inRange {max=500}}
    boneToModify {boneName=dyn_pant_leg_r_f_01}
    rotation {pitch=5}
   }
   n47 = "Transform (Modify) Bone - Bone: dyn_pant_leg_r_mf_01"!(ComponentPose=@n46, Alpha:float=@n33, Rotation="5.000,0.008,0.001") -> (Pose) {
    rotationMode=BMM_Additive
    rotationSpace=BCS_BoneSpace
    translationMode=BMM_Additive
    translationSpace=BCS_BoneSpace
    alphaBoolBlend {blendOption=Linear}
    alphaScaleBiasClamp {bInterpResult=true, interpSpeedDecreasing=5, interpSpeedIncreasing=7, inRange {max=500}}
    boneToModify {boneName=dyn_pant_leg_r_mf_01}
    rotation=<rot 5 0.008 0.001>
   }
   n48 = "Transform (Modify) Bone - Bone: dyn_pant_leg_r_m_01"!(ComponentPose=@n47, Alpha:float=@n33, Rotation="0,0,0") -> (Pose) {
    rotationMode=BMM_Additive
    rotationSpace=BCS_BoneSpace
    translationMode=BMM_Additive
    translationSpace=BCS_BoneSpace
    alphaBoolBlend {blendOption=Linear}
    alphaScaleBiasClamp {bInterpResult=true, interpSpeedDecreasing=5, interpSpeedIncreasing=7, inRange {max=500}}
    boneToModify {boneName=dyn_pant_leg_r_m_01}
   }
   n49 = "Transform (Modify) Bone - Bone: dyn_pant_leg_r_b_01_Base"!(ComponentPose=@n48, Alpha=1, Rotation="-20.000,-0.000,-0.000") -> (Pose) {
    rotationMode=BMM_Additive
    rotationSpace=BCS_BoneSpace
    translationMode=BMM_Additive
    translationSpace=BCS_BoneSpace
    alphaBoolBlend {blendOption=Linear}
    alphaScaleBiasClamp {bInterpResult=true, interpSpeedDecreasing=5, interpSpeedIncreasing=7, inRange {max=500}}
    boneToModify {boneName=dyn_pant_leg_r_b_01_Base}
    rotation=<rot -20 0 0>
   }
   n50 = "Transform (Modify) Bone - Bone: dyn_pant_leg_r_mb_01_Base"!(ComponentPose=@n49, Alpha=1, Rotation="-10,0,0") -> (Pose) {
    rotationMode=BMM_Additive
    rotationSpace=BCS_BoneSpace
    translationMode=BMM_Additive
    translationSpace=BCS_BoneSpace
    alphaBoolBlend {blendOption=Linear}
    alphaScaleBiasClamp {bInterpResult=true, interpSpeedDecreasing=5, interpSpeedIncreasing=7, inRange {max=500}}
    boneToModify {boneName=dyn_pant_leg_r_mb_01_Base}
    rotation {pitch=-10}
   }
   n51 = "Transform (Modify) Bone - Bone: dyn_pant_leg_l_f_01"!(ComponentPose=@n50, Alpha:float=@n33, Rotation="10,0,0") -> (Pose) {
    rotationMode=BMM_Additive
    rotationSpace=BCS_BoneSpace
    translationMode=BMM_Additive
    translationSpace=BCS_BoneSpace
    alphaBoolBlend {blendOption=Linear}
    alphaScaleBiasClamp {bInterpResult=true, interpSpeedDecreasing=5, interpSpeedIncreasing=7, inRange {max=500}}
    boneToModify {boneName=dyn_pant_leg_l_f_01}
    rotation {pitch=10}
   }
   n52 = "Transform (Modify) Bone - Bone: dyn_pant_leg_l_mf_01"!(ComponentPose=@n51, Alpha:float=@n33, Rotation="10,0,0") -> (Pose) {
    rotationMode=BMM_Additive
    rotationSpace=BCS_BoneSpace
    translationMode=BMM_Additive
    translationSpace=BCS_BoneSpace
    alphaBoolBlend {blendOption=Linear}
    alphaScaleBiasClamp {bInterpResult=true, interpSpeedDecreasing=5, interpSpeedIncreasing=7, inRange {max=500}}
    boneToModify {boneName=dyn_pant_leg_l_mf_01}
    rotation {pitch=10}
   }
   n53 = "Transform (Modify) Bone - Bone: dyn_pant_leg_l_m_01"!(ComponentPose=@n52, Alpha=@n33, Rotation="10,0,0") -> (Pose) {
    rotationMode=BMM_Additive
    rotationSpace=BCS_BoneSpace
    translationMode=BMM_Additive
    translationSpace=BCS_BoneSpace
    alphaBoolBlend {blendOption=Linear}
    alphaScaleBiasClamp {bInterpResult=true, interpSpeedDecreasing=5, interpSpeedIncreasing=7, inRange {max=500}}
    boneToModify {boneName=dyn_pant_leg_l_m_01}
    rotation {pitch=10}
   }
   n54 = "Transform (Modify) Bone - Bone: dyn_pant_leg_l_b_01_Base"!(ComponentPose=@n53, Alpha=1, Rotation="15,0,0") -> (Pose) {
    rotationMode=BMM_Additive
    rotationSpace=BCS_BoneSpace
    translationMode=BMM_Additive
    translationSpace=BCS_BoneSpace
    alphaBoolBlend {blendOption=Linear}
    alphaScaleBiasClamp {bInterpResult=true, interpSpeedDecreasing=5, interpSpeedIncreasing=7, inRange {max=500}}
    boneToModify {boneName=dyn_pant_leg_l_b_01_Base}
    rotation {pitch=15}
   }
   n55 = "Transform (Modify) Bone - Bone: dyn_pant_leg_l_mb_01_Base"!(ComponentPose=@n54, Alpha=1, Rotation="10,-10,0") -> (Pose) {
    rotationMode=BMM_Additive
    rotationSpace=BCS_BoneSpace
    translationMode=BMM_Additive
    translationSpace=BCS_BoneSpace
    alphaBoolBlend {blendOption=Linear}
    alphaScaleBiasClamp {bInterpResult=true, interpSpeedDecreasing=5, interpSpeedIncreasing=7, inRange {max=500}}
    boneToModify {boneName=dyn_pant_leg_l_mb_01_Base}
    rotation {pitch=10, yaw=-10}
   }
   n56 = "Transform (Modify) Bone - Bone: dyn_pant_leg_l_mb_03"!(ComponentPose=@n55, Alpha=@n33, Rotation="-20,0,20") -> (Pose) {
    rotationMode=BMM_Additive
    rotationSpace=BCS_BoneSpace
    translationMode=BMM_Additive
    translationSpace=BCS_BoneSpace
    alphaBoolBlend {blendOption=Linear}
    alphaScaleBiasClamp {bInterpResult=true, interpSpeedDecreasing=5, interpSpeedIncreasing=7, inRange {max=500}}
    boneToModify {boneName=dyn_pant_leg_l_mb_03}
    rotation {pitch=-20, roll=20}
   }
   n57 = Blend Poses by bool(BlendPose_0=Component To Local(ComponentPose=@n56), BlendPose_1=Component To Local(ComponentPose=@n32), BlendTime_0=0.25, BlendTime_1=0.25) {blendType=Linear}
   n58 = Get KneeSliding() -> (KneeSliding:double)
   n59 = Get KneeSliding()
   n60 = Get speedINV() -> (speedINV:double)
   n61 = "Transform (Modify) Bone - Bone: dyn_pant_leg_r_m_02"!(
    ComponentPose=Local To Component(LocalPose=Use cached pose 'CRCache_01'())
    Alpha:float=1
    Translation="-3,0,0"
    Rotation="0,0,0"
    Scale:Vector="1,1,1"
   ) -> (Pose) {translationMode=BMM_Additive, alphaBoolBlend {blendOption=Linear}, boneToModify {boneName=dyn_pant_leg_r_m_02}, translation {x=-3}}
   n62 = "Transform (Modify) Bone - Bone: dyn_pant_leg_l_m_01"!(ComponentPose=@n61, Alpha=1, Translation:Vector="2,0,0", Rotation="0,-2,20", Scale:Vector="1,1,1") -> (Pose) {
    rotationMode=BMM_Additive
    translationMode=BMM_Additive
    alphaBoolBlend {blendOption=Linear}
    boneToModify {boneName=dyn_pant_leg_l_m_01}
    rotation {roll=20, yaw=-2}
    translation {x=2}
   }
   n63 = "Transform (Modify) Bone - Bone: dyn_pant_leg_l_mf_01"!(ComponentPose=@n62, Alpha=1, Translation:Vector="0,0,0", Rotation="0,-10,10", Scale:Vector="1,1,1") -> (Pose) {
    rotationMode=BMM_Additive
    translationMode=BMM_Additive
    alphaBoolBlend {blendOption=Linear}
    boneToModify {boneName=dyn_pant_leg_l_mf_01}
    rotation {roll=10, yaw=-10}
   }
   n64 = Look At - Bone: dyn_pant_leg_l_mb_01_Base(ComponentPose=@n63, Alpha=1) {
    alphaBoolBlend {blendOption=Linear}
    boneToModify {boneName=dyn_pant_leg_l_mb_01_Base}
    lookAtLocation {x=5, z=-12}
    lookAtTarget {boneReference {boneName=calf_twist_01_l}}
    lookAt_Axis {axis=<1 0>}
    lookUp_Axis {axis=<0 1>}
   }
   n65 = Look At - Bone: dyn_pant_leg_l_f_01(ComponentPose=@n64, Alpha=@n60) {
    alphaBoolBlend {blendOption=Linear}
    boneToModify {boneName=dyn_pant_leg_l_f_01}
    lookAtLocation {x=13.5}
    lookAtTarget {boneReference {boneName=calf_l}}
    lookAt_Axis {axis=<1 0>}
    lookUp_Axis {axis=<0 1>}
   }
   n66 = Look At - Bone: dyn_pant_leg_l_f_02(ComponentPose=@n65, Alpha=@n60) {
    alphaBoolBlend {blendOption=Linear}
    boneToModify {boneName=dyn_pant_leg_l_f_02}
    lookAtLocation {x=11}
    lookAtTarget {boneReference {boneName=calf_l}}
    lookAt_Axis {axis=<1 0>}
    lookUp_Axis {axis=<0 1>}
   }
   n67 = Look At - Bone: dyn_pant_leg_l_f_03(ComponentPose=@n66, Alpha=@n60) {
    alphaBoolBlend {blendOption=Linear}
    boneToModify {boneName=dyn_pant_leg_l_f_03}
    lookAtLocation {x=5}
    lookAtTarget {boneReference {boneName=calf_l}}
    lookAt_Axis {axis=<1 0>}
    lookUp_Axis {axis=<0 1>}
   }
   n68 = "Transform (Modify) Bone - Bone: dyn_pant_leg_l_mb_01_Base"!(ComponentPose=@n67, Alpha=Get Crouching() -> (Crouching:double), Translation:Vector="0,0,0", Rotation="0,0,20", Scale:Vector="1,1,1") -> (Pose) {
    rotationMode=BMM_Additive
    translationMode=BMM_Additive
    alphaBoolBlend {blendOption=Linear}
    alphaScaleBiasClamp {bInterpResult=true}
    boneToModify {boneName=dyn_pant_leg_l_mb_01_Base}
    rotation {roll=20}
   }
   n69 = "Transform (Modify) Bone - Bone: dyn_pant_leg_l_b_01_Base"!(ComponentPose=@n68, bAlphaBoolEnabled:bool=True, Translation:Vector="0,0.000,-3", Rotation="0,0,40", Scale:Vector="1,1,1") -> (Pose) {
    alphaInputType=Bool
    rotationMode=BMM_Additive
    translationMode=BMM_Additive
    alphaBoolBlend {blendOption=Linear}
    alphaScaleBiasClamp {bInterpResult=true}
    boneToModify {boneName=dyn_pant_leg_l_b_01_Base}
    rotation {roll=40}
    translation {z=-3}
   }
   n70 = "Constraint - dyn_pant_leg_r_m_03 (R)"(ComponentPose=@n69, Alpha=0.3, ConstraintWeights_0=1) {
    alphaBoolBlend {blendOption=Linear}
    boneToModify {boneName=dyn_pant_leg_r_m_03}
    constraintSetup [{transformType=Rotation, targetBone {boneName=calf_r}}]
    constraintWeights [1]
   }
   n71 = "Constraint - dyn_pant_leg_r_m_02 (R)"(ComponentPose=@n70, Alpha=0.3, ConstraintWeights_0=1) {
    alphaBoolBlend {blendOption=Linear}
    boneToModify {boneName=dyn_pant_leg_r_m_02}
    constraintSetup [{transformType=Rotation, targetBone {boneName=calf_r}}]
    constraintWeights [1]
   }
   n72 = "Constraint - dyn_pant_leg_r_mb_02 (R)"(ComponentPose=@n71, Alpha=0.75, ConstraintWeights_0=1) {
    alphaBoolBlend {blendOption=Linear}
    boneToModify {boneName=dyn_pant_leg_r_mb_02}
    constraintSetup [{transformType=Rotation, targetBone {boneName=calf_r}}]
    constraintWeights [1]
   }
   n73 = "Constraint - dyn_pant_leg_r_mb_02 (R)"(ComponentPose=@n72, Alpha=0.4, ConstraintWeights_0=1) {
    alphaBoolBlend {blendOption=Linear}
    boneToModify {boneName=dyn_pant_leg_r_mb_02}
    constraintSetup [{transformType=Rotation, targetBone {boneName=calf_r}}]
    constraintWeights [1]
   }
   n74 = "Constraint - dyn_pant_leg_r_b_03 (R)"(ComponentPose=@n73, Alpha=0.6, ConstraintWeights_0=1) {
    alphaBoolBlend {blendOption=Linear}
    boneToModify {boneName=dyn_pant_leg_r_b_03}
    constraintSetup [{transformType=Rotation, targetBone {boneName=calf_r}}]
    constraintWeights [1]
   }
   n75 = Get Crouching()
   n76 = Anim Dynamics - Bone: dyn_pant_leg_r_mf_03(ComponentPose=@n74, Alpha=@n75, ExternalForce="0,0,0") {
    bGravityOverrideInSimSpace=true
    bUseGravityOverride=true
    simulationSpace=RootRelative
    alphaBoolBlend {blendInTime=0.1, blendOption=Linear, blendOutTime=0.1}
    alphaScaleBiasClamp {bInterpResult=true}
    boundBone {boneName=dyn_pant_leg_r_mf_03}
    componentAppliedLinearAccClamp=<10000 10000>
    componentLinearAccScale=<0.5 0.5 0.5>
    gravityOverride {z=-980}
    physicsBodyDefinitions [{
      sphereCollisionRadius=0
      boundBone {boneName=dyn_pant_leg_r_mf_03}
      boxExtents=<5 5 2.5>
      constraintSetup {angularLimitsMax=<0 2 20>, angularLimitsMin=<0 -2 -2>, angularTarget=<1 0 0>}
      localJointOffset=<-0.1 0 0>
     }]
    relativeSpaceBone {boneName=root}
   }
   n77 = Anim Dynamics - Bone: dyn_pant_leg_r_mf_04(ComponentPose=@n76, Alpha=@n75, ExternalForce="0,0,0") {
    bGravityOverrideInSimSpace=true
    bUseGravityOverride=true
    simulationSpace=RootRelative
    alphaBoolBlend {blendInTime=0.1, blendOption=Linear, blendOutTime=0.1}
    alphaScaleBiasClamp {bInterpResult=true}
    boundBone {boneName=dyn_pant_leg_r_mf_04}
    componentAppliedLinearAccClamp=<10000 10000>
    componentLinearAccScale=<0.5 0.5 0.5>
    gravityOverride {z=-980}
    physicsBodyDefinitions [{
      sphereCollisionRadius=0
      boundBone {boneName=dyn_pant_leg_r_mf_04}
      boxExtents=<5 5 2.5>
      constraintSetup {angularLimitsMax=<0 2 20>, angularLimitsMin=<0 -2 -2>, angularTarget=<1 0 0>}
      localJointOffset=<-0.1 0 0>
     }]
    relativeSpaceBone {boneName=root}
   }
   n78 = Anim Dynamics - Bone: dyn_pant_leg_r_m_03(ComponentPose=@n77, Alpha=@n75, ExternalForce="0,0,0") {
    bGravityOverrideInSimSpace=true
    bUseGravityOverride=true
    simulationSpace=RootRelative
    alphaBoolBlend {blendInTime=0.1, blendOption=Linear, blendOutTime=0.1}
    alphaScaleBiasClamp {bInterpResult=true}
    boundBone {boneName=dyn_pant_leg_r_m_03}
    componentAppliedLinearAccClamp=<10000 10000>
    componentLinearAccScale=<0.5 0.5 0.5>
    gravityOverride {z=-980}
    physicsBodyDefinitions [{
      sphereCollisionRadius=0
      boundBone {boneName=dyn_pant_leg_r_m_03}
      boxExtents=<5 5 2.5>
      constraintSetup {angularLimitsMax=<0 2 20>, angularLimitsMin=<0 -2 -20>, angularTarget=<1 0 0>}
      localJointOffset=<-0.1 0 0>
     }]
    relativeSpaceBone {boneName=root}
   }
   n79 = Anim Dynamics - Bone: dyn_pant_leg_r_m_04(ComponentPose=@n78, Alpha=@n75, ExternalForce="0,0,0") {
    bGravityOverrideInSimSpace=true
    bUseGravityOverride=true
    simulationSpace=RootRelative
    alphaBoolBlend {blendInTime=0.1, blendOption=Linear, blendOutTime=0.1}
    alphaScaleBiasClamp {bInterpResult=true}
    boundBone {boneName=dyn_pant_leg_r_m_04}
    componentAppliedLinearAccClamp=<10000 10000>
    componentLinearAccScale=<0.5 0.5 0.5>
    gravityOverride {z=-980}
    physicsBodyDefinitions [{
      sphereCollisionRadius=0
      boundBone {boneName=dyn_pant_leg_r_m_04}
      boxExtents=<5 5 2.5>
      constraintSetup {angularLimitsMax=<0 2 20>, angularLimitsMin=<0 -2 -20>, angularTarget=<1 0 0>}
      localJointOffset=<-0.1 0 0>
     }]
    relativeSpaceBone {boneName=root}
   }
   n80 = Anim Dynamics - Bone: dyn_pant_leg_r_mb_03(ComponentPose=@n79, Alpha=@n75, ExternalForce="0,0,0") {
    bGravityOverrideInSimSpace=true
    bUseGravityOverride=true
    simulationSpace=RootRelative
    alphaBoolBlend {blendInTime=0.1, blendOption=Linear, blendOutTime=0.1}
    alphaScaleBiasClamp {bInterpResult=true}
    boundBone {boneName=dyn_pant_leg_r_mb_03}
    componentAppliedLinearAccClamp=<10000 10000>
    componentLinearAccScale=<0.5 0.5 0.5>
    gravityOverride {z=-980}
    physicsBodyDefinitions [{
      sphereCollisionRadius=0
      boundBone {boneName=dyn_pant_leg_r_mb_03}
      boxExtents=<5 5 2.5>
      constraintSetup {angularLimitsMax=<0 2 20>, angularLimitsMin=<0 -2 -20>, angularTarget=<1 0 0>}
      localJointOffset=<-0.1 0 0>
     }]
    relativeSpaceBone {boneName=root}
   }
   n81 = Anim Dynamics - Bone: dyn_pant_leg_r_mb_04(ComponentPose=@n80, Alpha=@n75, ExternalForce="0,0,0") {
    bGravityOverrideInSimSpace=true
    bUseGravityOverride=true
    simulationSpace=RootRelative
    alphaBoolBlend {blendInTime=0.1, blendOption=Linear, blendOutTime=0.1}
    alphaScaleBiasClamp {bInterpResult=true}
    boundBone {boneName=dyn_pant_leg_r_mb_04}
    componentAppliedLinearAccClamp=<10000 10000>
    componentLinearAccScale=<0.5 0.5 0.5>
    gravityOverride {z=-980}
    physicsBodyDefinitions [{
      sphereCollisionRadius=0
      boundBone {boneName=dyn_pant_leg_r_mb_04}
      boxExtents=<5 5 2.5>
      constraintSetup {angularLimitsMax=<0 2 20>, angularLimitsMin=<0 -2 -20>, angularTarget=<1 0 0>}
      localJointOffset=<-0.1 0 0>
     }]
    relativeSpaceBone {boneName=root}
   }
   n82 = "Transform (Modify) Bone - Bone: dyn_pant_leg_r_mb_02"(ComponentPose=@n81, Alpha=@n75, Translation="-1.661,5.021,-3.481", Rotation="3.702,-16.058,-43.009", Scale="1,1,1") {
    rotationMode=BMM_Additive
    rotationSpace=BCS_BoneSpace
    translationMode=BMM_Additive
    translationSpace=BCS_BoneSpace
    alphaBoolBlend {blendOption=Linear}
    alphaScaleBiasClamp {bInterpResult=true}
    boneToModify {boneName=dyn_pant_leg_r_mb_02}
    rotation=<rot 3.702 -16.058 -43.009>
    translation=<-1.661 5.021 -3.481>
   }
   n83 = "Transform (Modify) Bone - Bone: dyn_pant_leg_r_m_02"(ComponentPose=@n82, Alpha=@n75, Translation="1.004,0.268,-2.077", Rotation="26.421,6.268,2.946", Scale="1,1,1") {
    rotationMode=BMM_Additive
    rotationSpace=BCS_BoneSpace
    translationMode=BMM_Additive
    translationSpace=BCS_BoneSpace
    alphaBoolBlend {blendOption=Linear}
    alphaScaleBiasClamp {bInterpResult=true}
    boneToModify {boneName=dyn_pant_leg_r_m_02}
    rotation=<rot 26.421 6.268 2.946>
    translation=<1.004 0.268 -2.077>
   }
   n84 = "Transform (Modify) Bone - Bone: dyn_pant_leg_r_b_01_Base"!(ComponentPose=@n83, Alpha=@n75, Translation:Vector="0,-2.5,0", Rotation="74.151,-87.192,-71.568", Scale:Vector="1,1,1") -> (Pose) {
    rotationMode=BMM_Additive
    rotationSpace=BCS_BoneSpace
    translationMode=BMM_Additive
    translationSpace=BCS_BoneSpace
    alphaBoolBlend {blendOption=Linear}
    alphaScaleBiasClamp {bInterpResult=true}
    boneToModify {boneName=dyn_pant_leg_r_b_01_Base}
    rotation=<rot 74.151 -87.1919 -71.5679>
    translation {y=-2.5}
   }
   n85 = "Transform (Modify) Bone - Bone: dyn_pant_leg_r_b_01_Base"(ComponentPose=@n84, Alpha=@n59, Translation="-0.078,9.694,14.129", Rotation="0,0,0", Scale="1,1,1") {
    rotationMode=BMM_Additive
    rotationSpace=BCS_BoneSpace
    translationMode=BMM_Additive
    translationSpace=BCS_BoneSpace
    alphaBoolBlend {blendOption=Linear}
    alphaScaleBiasClamp {bInterpResult=true}
    boneToModify {boneName=dyn_pant_leg_r_b_01_Base}
    translation=<-0.078 9.694 14.129>
   }
   n86 = "Transform (Modify) Bone - Bone: dyn_pant_leg_r_mb_01_Base"!(ComponentPose=@n85, Alpha=@n59, Translation:Vector="3.984,-2.680,1.113", Rotation="-3.627,-11.794,2.258", Scale:Vector="1,1,1") -> (Pose) {
    rotationMode=BMM_Additive
    rotationSpace=BCS_BoneSpace
    translationMode=BMM_Additive
    translationSpace=BCS_BoneSpace
    alphaBoolBlend {blendOption=Linear}
    alphaScaleBiasClamp {bInterpResult=true}
    boneToModify {boneName=dyn_pant_leg_r_mb_01_Base}
    rotation=<rot -3.627 -11.794 2.258>
    translation=<3.984 -2.68 1.113>
   }
   n87 = "Transform (Modify) Bone - Bone: dyn_pant_leg_r_m_01"!(ComponentPose=@n86, Alpha=@n59, Translation:Vector="0,5,0", Rotation="0,40,0", Scale:Vector="1,1,1") -> (Pose) {
    rotationMode=BMM_Additive
    rotationSpace=BCS_BoneSpace
    translationMode=BMM_Additive
    translationSpace=BCS_BoneSpace
    alphaBoolBlend {blendOption=Linear}
    alphaScaleBiasClamp {bInterpResult=true}
    boneToModify {boneName=dyn_pant_leg_r_m_01}
    rotation {yaw=40}
    translation {y=5}
   }
   n88 = Get Crouching()
   n89 = Anim Dynamics - Bone: dyn_pant_leg_l_mf_03(ComponentPose=@n87, Alpha=@n88, ExternalForce="0,0,0") {
    angularDampingOverride=0.9
    angularSpringConstant=700
    bAngularSpring=true
    bGravityOverrideInSimSpace=true
    bOverrideLinearDamping=true
    bUseGravityOverride=true
    simulationSpace=RootRelative
    alphaBoolBlend {blendInTime=0.1, blendOption=Linear, blendOutTime=0.1}
    alphaScaleBiasClamp {bInterpResult=true}
    boundBone {boneName=dyn_pant_leg_l_mf_03}
    componentAppliedLinearAccClamp=<10000 10000>
    componentLinearAccScale=<0.5 0.5 0.5>
    gravityOverride {z=-980}
    physicsBodyDefinitions [{
      sphereCollisionRadius=0
      boundBone {boneName=dyn_pant_leg_l_mf_03}
      boxExtents=<5 5 2.5>
      constraintSetup {angularLimitsMax=<0 2 20>, angularLimitsMin=<0 -2 -2>, angularTarget=<1 0 0>}
      localJointOffset=<0.1 0 0>
     }]
    relativeSpaceBone {boneName=root}
   }
   n90 = Anim Dynamics - Bone: dyn_pant_leg_l_mf_04(ComponentPose=@n89, Alpha=@n88, ExternalForce="0,0,0") {
    angularDampingOverride=0.9
    angularSpringConstant=700
    bAngularSpring=true
    bGravityOverrideInSimSpace=true
    bOverrideLinearDamping=true
    bUseGravityOverride=true
    simulationSpace=RootRelative
    alphaBoolBlend {blendInTime=0.1, blendOption=Linear, blendOutTime=0.1}
    alphaScaleBiasClamp {bInterpResult=true}
    boundBone {boneName=dyn_pant_leg_l_mf_04}
    componentAppliedLinearAccClamp=<10000 10000>
    componentLinearAccScale=<0.5 0.5 0.5>
    gravityOverride {z=-980}
    physicsBodyDefinitions [{
      sphereCollisionRadius=0
      boundBone {boneName=dyn_pant_leg_l_mf_04}
      boxExtents=<5 5 2.5>
      constraintSetup {angularLimitsMax=<0 2 20>, angularLimitsMin=<0 -2 -2>, angularTarget=<1 0 0>}
      localJointOffset=<0.1 0 0>
     }]
    relativeSpaceBone {boneName=root}
   }
   n91 = Anim Dynamics - Bone: dyn_pant_leg_l_m_03(ComponentPose=@n90, Alpha=@n88, ExternalForce="0,0,0") {
    bGravityOverrideInSimSpace=true
    bUseGravityOverride=true
    simulationSpace=RootRelative
    alphaBoolBlend {blendInTime=0.1, blendOption=Linear, blendOutTime=0.1}
    alphaScaleBiasClamp {bInterpResult=true}
    boundBone {boneName=dyn_pant_leg_l_m_03}
    componentAppliedLinearAccClamp=<10000 10000>
    componentLinearAccScale=<0.5 0.5 0.5>
    gravityOverride {z=-980}
    physicsBodyDefinitions [{
      sphereCollisionRadius=0
      boundBone {boneName=dyn_pant_leg_l_m_03}
      boxExtents=<5 5 2.5>
      constraintSetup {angularLimitsMax=<0 2 10>, angularLimitsMin=<0 -2 -10>, angularTarget=<1 0 0>}
      localJointOffset=<0.1 0 0>
     }]
    relativeSpaceBone {boneName=root}
   }
   n92 = Anim Dynamics - Bone: dyn_pant_leg_l_m_04(ComponentPose=@n91, Alpha=@n88, ExternalForce="0,0,0") {
    bGravityOverrideInSimSpace=true
    bUseGravityOverride=true
    simulationSpace=RootRelative
    alphaBoolBlend {blendInTime=0.1, blendOption=Linear, blendOutTime=0.1}
    alphaScaleBiasClamp {bInterpResult=true}
    boundBone {boneName=dyn_pant_leg_l_m_04}
    componentAppliedLinearAccClamp=<10000 10000>
    componentLinearAccScale=<0.5 0.5 0.5>
    gravityOverride {z=-980}
    physicsBodyDefinitions [{
      sphereCollisionRadius=0
      boundBone {boneName=dyn_pant_leg_l_m_04}
      boxExtents=<5 5 2.5>
      constraintSetup {angularLimitsMax=<0 2 10>, angularLimitsMin=<0 -10 -10>, angularTarget=<1 0 0>}
      localJointOffset=<0.1 0 0>
     }]
    relativeSpaceBone {boneName=root}
   }
   n93 = Anim Dynamics - Bone: dyn_pant_leg_l_mb_03(ComponentPose=@n92, Alpha=@n88, ExternalForce="0,0,0") {
    angularDampingOverride=0.9
    angularSpringConstant=700
    bAngularSpring=true
    bGravityOverrideInSimSpace=true
    bOverrideLinearDamping=true
    bUseGravityOverride=true
    simulationSpace=RootRelative
    alphaBoolBlend {blendInTime=0.1, blendOption=Linear, blendOutTime=0.1}
    alphaScaleBiasClamp {bInterpResult=true}
    boundBone {boneName=dyn_pant_leg_l_mb_03}
    componentAppliedLinearAccClamp=<10000 10000>
    componentLinearAccScale=<0.5 0.5 0.5>
    gravityOverride {z=-980}
    physicsBodyDefinitions [{
      sphereCollisionRadius=0
      boundBone {boneName=dyn_pant_leg_l_mb_03}
      boxExtents=<5 5 2.5>
      constraintSetup {angularLimitsMax=<0 0 5>, angularLimitsMin=<0 0 -5>, angularTarget=<1 0 0>}
      localJointOffset=<0.1 0 0>
     }]
    relativeSpaceBone {boneName=root}
   }
   n94 = Anim Dynamics - Bone: dyn_pant_leg_l_mb_04(ComponentPose=@n93, Alpha=@n88, ExternalForce="0,0,0") {
    angularDampingOverride=0.9
    angularSpringConstant=700
    bAngularSpring=true
    bGravityOverrideInSimSpace=true
    bOverrideAngularDamping=true
    bUseGravityOverride=true
    simulationSpace=RootRelative
    alphaBoolBlend {blendInTime=0.1, blendOption=Linear, blendOutTime=0.1}
    alphaScaleBiasClamp {bInterpResult=true}
    boundBone {boneName=dyn_pant_leg_l_mb_04}
    componentAppliedLinearAccClamp=<10000 10000>
    componentLinearAccScale=<0.5 0.5 0.5>
    gravityOverride {z=-980}
    physicsBodyDefinitions [{
      sphereCollisionRadius=0
      boundBone {boneName=dyn_pant_leg_l_mb_04}
      boxExtents=<5 5 2.5>
      constraintSetup {angularLimitsMax=<0 0 5>, angularLimitsMin=<0 0 -5>, angularTarget=<1 0 0>}
      localJointOffset=<0.1 0 0>
     }]
    relativeSpaceBone {boneName=root}
   }
   n95 = Anim Dynamics - Bone: dyn_pant_leg_l_b_02(ComponentPose=@n94, Alpha=@n88, ExternalForce="0,0,0") {
    bGravityOverrideInSimSpace=true
    bUseGravityOverride=true
    simulationSpace=RootRelative
    alphaBoolBlend {blendInTime=0.1, blendOption=Linear, blendOutTime=0.1}
    alphaScaleBiasClamp {bInterpResult=true}
    boundBone {boneName=dyn_pant_leg_l_b_02}
    componentAppliedLinearAccClamp=<10000 10000>
    componentLinearAccScale=<0.5 0.5 0.5>
    gravityOverride {z=-980}
    physicsBodyDefinitions [{
      sphereCollisionRadius=0
      boundBone {boneName=dyn_pant_leg_l_b_02}
      boxExtents=<5 5 2.5>
      constraintSetup {angularLimitsMax=<0 5 0>, angularLimitsMin=<0 -5 0>, angularTarget=<1 0 0>}
      localJointOffset=<0.1 0 0>
     }]
    relativeSpaceBone {boneName=root}
   }
   n96 = Anim Dynamics - Bone: dyn_pant_leg_l_b_03(ComponentPose=@n95, Alpha=@n88, ExternalForce="0,0,0") {
    bGravityOverrideInSimSpace=true
    bUseGravityOverride=true
    simulationSpace=RootRelative
    alphaBoolBlend {blendInTime=0.1, blendOption=Linear, blendOutTime=0.1}
    alphaScaleBiasClamp {bInterpResult=true}
    boundBone {boneName=dyn_pant_leg_l_b_03}
    componentAppliedLinearAccClamp=<10000 10000>
    componentLinearAccScale=<0.5 0.5 0.5>
    gravityOverride {z=-980}
    physicsBodyDefinitions [{
      sphereCollisionRadius=0
      boundBone {boneName=dyn_pant_leg_l_b_03}
      boxExtents=<5 5 2.5>
      constraintSetup {angularLimitsMax=<0 2 20>, angularLimitsMin=<0 -2 -20>, angularTarget=<1 0 0>}
      localJointOffset=<0.1 0 0>
     }]
    relativeSpaceBone {boneName=root}
   }
   n97 = "Transform (Modify) Bone - Bone: dyn_pant_leg_l_b_01_Base"!(ComponentPose=@n96, Alpha:float=@n88, Translation="0.25,1.5,-2", Rotation="0,0,-35", Scale="1,1,1") -> (Pose) {
    rotationMode=BMM_Additive
    rotationSpace=BCS_BoneSpace
    translationMode=BMM_Additive
    translationSpace=BCS_BoneSpace
    alphaBoolBlend {blendOption=Linear}
    alphaScaleBiasClamp {bInterpResult=true}
    boneToModify {boneName=dyn_pant_leg_l_b_01_Base}
    rotation {roll=-35}
    translation=<0.25 1.5 -2>
   }
   n98 = "Transform (Modify) Bone - Bone: dyn_pant_leg_l_b_01_Base"(ComponentPose=@n97, Alpha=@n58, Translation="0,0,0", Rotation="50.675,-157.525,45.253", Scale="1,1,1") {
    rotationMode=BMM_Additive
    rotationSpace=BCS_BoneSpace
    translationMode=BMM_Additive
    translationSpace=BCS_BoneSpace
    alphaBoolBlend {blendOption=Linear}
    alphaScaleBiasClamp {bInterpResult=true}
    boneToModify {boneName=dyn_pant_leg_l_b_01_Base}
    rotation=<rot 50.675 -157.525 45.253>
   }
   n99 = "Transform (Modify) Bone - Bone: dyn_pant_leg_l_mb_01_Base"(ComponentPose=@n98, Alpha=@n58, Translation="-2.493,-0.582,-11.449", Rotation="81.708,-59.796,43.993", Scale="1,1,1") {
    rotationMode=BMM_Additive
    rotationSpace=BCS_BoneSpace
    translationMode=BMM_Additive
    translationSpace=BCS_BoneSpace
    alphaBoolBlend {blendOption=Linear}
    alphaScaleBiasClamp {bInterpResult=true}
    boneToModify {boneName=dyn_pant_leg_l_mb_01_Base}
    rotation=<rot 81.708 -59.796 43.993>
    translation=<-2.493 -0.582 -11.449>
   }
   n100 = "Transform (Modify) Bone - Bone: dyn_pant_leg_l_m_01"(ComponentPose=@n99, Alpha=@n58, Translation="-15,-1.35,-6.3", Rotation="45,5,-13.5", Scale="1,1,1") {
    rotationMode=BMM_Additive
    rotationSpace=BCS_BoneSpace
    translationMode=BMM_Additive
    translationSpace=BCS_BoneSpace
    alphaBoolBlend {blendOption=Linear}
    alphaScaleBiasClamp {bInterpResult=true}
    boneToModify {boneName=dyn_pant_leg_l_m_01}
    rotation=<rot 45 5 -13.5>
    translation=<-15 -1.35 -6.3>
   }
   n101 = "Transform (Modify) Bone - Bone: dyn_pant_leg_l_f_01"!(ComponentPose=@n100, Alpha=@n58, Translation:Vector="0,0,0", Rotation="2.981,-8.627,-1.642", Scale:Vector="1,1,1") -> (Pose) {
    rotationMode=BMM_Additive
    rotationSpace=BCS_BoneSpace
    translationMode=BMM_Additive
    translationSpace=BCS_BoneSpace
    alphaBoolBlend {blendOption=Linear}
    alphaScaleBiasClamp {bInterpResult=true}
    boneToModify {boneName=dyn_pant_leg_l_f_01}
    rotation=<rot 2.981 -8.627 -1.642>
   }
   n102 = Get CrouchMove() -> (CrouchMove:double)
   n103 = "Transform (Modify) Bone - Bone: dyn_pant_leg_r_b_01_Base"(ComponentPose=@n101, Alpha=@n102, Translation="0,0,0", Rotation="-75.000,-0.000,0.000", Scale="1,1,1") {
    rotationMode=BMM_Additive
    rotationSpace=BCS_BoneSpace
    translationMode=BMM_Additive
    translationSpace=BCS_BoneSpace
    alphaBoolBlend {blendOption=Linear}
    alphaScaleBiasClamp {bInterpResult=true}
    boneToModify {boneName=dyn_pant_leg_r_b_01_Base}
    rotation=<rot -75 0 0>
   }
   n104 = "Transform (Modify) Bone - Bone: dyn_pant_leg_r_mb_01_Base"(ComponentPose=@n103, Alpha=@n102, Translation="0,0,0", Rotation="-3.802,26.127,-6.174", Scale="1,1,1") {
    rotationMode=BMM_Additive
    rotationSpace=BCS_BoneSpace
    translationMode=BMM_Additive
    translationSpace=BCS_BoneSpace
    alphaBoolBlend {blendOption=Linear}
    alphaScaleBiasClamp {bInterpResult=true}
    boneToModify {boneName=dyn_pant_leg_r_mb_01_Base}
    rotation=<rot -3.802 26.127 -6.174>
   }
   n105 = "Transform (Modify) Bone - Bone: dyn_pant_leg_r_mb_03"(ComponentPose=@n104, Alpha=@n102, Translation="3,-4,-4", Rotation="-24.994,10.000,-0.000", Scale="1,1,1") {
    rotationMode=BMM_Additive
    rotationSpace=BCS_BoneSpace
    translationMode=BMM_Additive
    translationSpace=BCS_BoneSpace
    alphaBoolBlend {blendOption=Linear}
    alphaScaleBiasClamp {bInterpResult=true}
    boneToModify {boneName=dyn_pant_leg_r_mb_03}
    rotation=<rot -24.994 10 0>
    translation=<3 -4 -4>
   }
   n106 = Component To Local(ComponentPose=@n105)
   n107 = Get bIsCrouching()
   n108 = Get isMoving() -> (isMoving:double)
   n109 = "Transform (Modify) Bone - Bone: dyn_pant_leg_r_mb_01_Base"!(
    ComponentPose=Local To Component(LocalPose=Blend Poses by bool(BlendPose_0=@n106, BlendPose_1=@n57, BlendTime_0=0.25, BlendTime_1=0.25, bActiveValue=@n107) {blendType=Linear})
    Alpha=@n108
    Rotation="2.5,0,0"
   ) -> (Pose) {
    rotationMode=BMM_Additive
    rotationSpace=BCS_BoneSpace
    translationMode=BMM_Additive
    translationSpace=BCS_BoneSpace
    alphaBoolBlend {blendOption=Linear}
    alphaScaleBiasClamp {bInterpResult=true}
    boneToModify {boneName=dyn_pant_leg_r_mb_01_Base}
    rotation {pitch=2.5}
   }
   n110 = "Transform (Modify) Bone - Bone: dyn_pant_leg_r_m_01"!(ComponentPose=@n109, Alpha=@n108, Rotation="5,0,0") -> (Pose) {
    rotationMode=BMM_Additive
    rotationSpace=BCS_BoneSpace
    translationMode=BMM_Additive
    translationSpace=BCS_BoneSpace
    alphaBoolBlend {blendOption=Linear}
    alphaScaleBiasClamp {bInterpResult=true, interpSpeedDecreasing=5, interpSpeedIncreasing=7}
    boneToModify {boneName=dyn_pant_leg_r_m_01}
    rotation {pitch=5}
   }
   n111 = "Transform (Modify) Bone - Bone: dyn_pant_leg_r_b_01_Base"!(ComponentPose=@n110, Alpha=@n108, Rotation="20,0,0") -> (Pose) {
    rotationMode=BMM_Additive
    rotationSpace=BCS_BoneSpace
    translationMode=BMM_Additive
    translationSpace=BCS_BoneSpace
    alphaBoolBlend {blendOption=Linear}
    alphaScaleBiasClamp {bInterpResult=true}
    boneToModify {boneName=dyn_pant_leg_r_b_01_Base}
    rotation {pitch=20}
   }
   Output Pose(Result=Component To Local(ComponentPose=@n111)) {layerGroup=None}
   n112 = Break Fort Anim Input Skydiving External Force(FortAnimInput_SkydivingExternalForce=Get SkydivingExternalForceInput()) -> (
    bUseSkydivingVectorForce=~
    SkydivingVectorForce=~
    HeadToPelvisDirection=~
    FloatingMultiplier=~
    FloatingAdditive=~
    DivingMultiplier=~
    DivingAdditive=~
    ParachutingMultiplier=~
    ParachutingAdditive=~
   )
   n113 = Anim Dynamics - Bone: dyn_belt_r(
    ComponentPose=Local To Component(LocalPose=Copy Pose From Mesh() {bUseAttachedParent=true})
    Alpha=0.4
    ExternalForce=@n112.SkydivingVectorForce
   ) {
    bGravityOverrideInSimSpace=true
    bUsePlanarLimit=false
    simulationSpace=RootRelative
    alphaBoolBlend {blendOption=Linear}
    boundBone {boneName=dyn_belt_r}
    componentAppliedLinearAccClamp=<10000 10000 10000>
    componentLinearAccScale=<1 1 1>
    componentLinearVelScale=<0.2 0.2 0.2>
    gravityOverride {z=-980}
    physicsBodyDefinitions [{
      collisionType=CustomSphere
      sphereCollisionRadius=4
      boundBone {boneName=dyn_belt_r}
      boxExtents=<15 15 5>
      constraintSetup {angularLimitsMax=<0 0 0>, angularLimitsMin=<0 0 0>, angularTarget=<1 0 0>, linearAxesMax=<0 0 0.5>, linearAxesMin=<-1 0 0>}
      localJointOffset=<0 0 -1>
     }]
   }
   n114 = Anim Dynamics - Bone: dyn_belt_l(ComponentPose=@n113, Alpha=0.4, ExternalForce=@n112.SkydivingVectorForce) {
    bGravityOverrideInSimSpace=true
    simulationSpace=RootRelative
    alphaBoolBlend {blendOption=Linear}
    boundBone {boneName=dyn_belt_l}
    componentAppliedLinearAccClamp=<10000 10000 10000>
    componentLinearAccScale=<1 1 1>
    componentLinearVelScale=<0.2 0.2 0.2>
    physicsBodyDefinitions [{
      sphereCollisionRadius=0
      boundBone {boneName=dyn_belt_l}
      boxExtents=<15 15 5>
      constraintSetup {angularLimitsMax=<0 0 0>, angularLimitsMin=<0 0 0>, angularTarget=<1 0 0>, linearAxesMax=<1 0 0.5>}
      localJointOffset=<0 0 -1>
     }]
   }
   n115 = RigidBody(ComponentPose=@n114) {
    bClampLinearTranslationLimitToRefPose=true
    bOverrideWorldGravity=true
    overridePhysicsAsset=&/F_MED_Masked_Warrior_Scarf_Physics
    simulationSpace=BaseBoneSpace
    alphaBoolBlend {blendOption=Linear}
    baseBoneRef {boneName=pelvis}
    componentAppliedLinearAccClamp=<1000 1000 1000>
    componentLinearVelScale=<0.5 0.5 0.5>
    overrideWorldGravity {z=-980}
   }
   n116 = Anim Dynamics - Bone: shawl_02(ComponentPose=@n115, Alpha=1, ExternalForce="0,0,0") {
    bGravityOverrideInSimSpace=true
    bUseGravityOverride=true
    simulationSpace=RootRelative
    alphaBoolBlend {blendOption=Linear}
    boundBone {boneName=shawl_02}
    componentAppliedLinearAccClamp=<5000 5000 5000>
    componentLinearAccScale=<0.1 0.1 0.1>
    componentLinearVelScale=<0.1 0.1 0.1>
    gravityOverride {z=-980}
    physicsBodyDefinitions [{
      sphereCollisionRadius=0
      boundBone {boneName=shawl_02}
      boxExtents=<5 8 1>
      constraintSetup {angularLimitsMax=<0 20 0>, angularLimitsMin=<0 -20 0>, angularTarget=<0 0 0>}
      localJointOffset=<2 0 0>
     }]
    relativeSpaceBone {boneName=root}
   }
   n117 = "Constraint - thigh_half_l (P)"(ComponentPose=@n116, Alpha=0, ConstraintWeights_0=1) {
    alphaBoolBlend {blendOption=Linear}
    boneToModify {boneName=thigh_half_l}
    constraintSetup [{targetBone {boneName=thigh_l}}]
    constraintWeights [1]
   }
   n118 = "Constraint - thigh_half_r (P)"(ComponentPose=@n117, Alpha=0.5, ConstraintWeights_0=1) {
    alphaBoolBlend {blendOption=Linear}
    boneToModify {boneName=thigh_half_r}
    constraintSetup [{targetBone {boneName=thigh_r}}]
    constraintWeights [1]
   }
   n119 = "Constraint - dyn_pant_leg_l_mb_01_Base (TP)"!(ComponentPose=@n118, Alpha=0.2, ConstraintWeights_0=0, ConstraintWeights_1:float=1) -> (Pose) {
    alphaBoolBlend {blendOption=Linear}
    boneToModify {boneName=dyn_pant_leg_l_mb_01_Base}
    constraintSetup [{transformType=Translation, targetBone {boneName=None}}, {targetBone {boneName=pelvis}}]
    constraintWeights [1,1]
   }
   n120 = "Constraint - dyn_pant_leg_r_b_01_Base (TP)"!(ComponentPose=@n119, Alpha=0, ConstraintWeights_0=0, ConstraintWeights_1:float=1) -> (Pose) {
    alphaBoolBlend {blendOption=Linear}
    boneToModify {boneName=dyn_pant_leg_r_b_01_Base}
    constraintSetup [{transformType=Translation, targetBone {boneName=None}}, {targetBone {boneName=pelvis}}]
    constraintWeights [1,1]
   }
   n121 = "Constraint - dyn_pant_leg_r_mb_01_Base (TP)"!(ComponentPose=@n120, Alpha=0, ConstraintWeights_0=0, ConstraintWeights_1:float=1) -> (Pose) {
    alphaBoolBlend {blendOption=Linear}
    boneToModify {boneName=dyn_pant_leg_r_mb_01_Base}
    constraintSetup [{transformType=Translation, targetBone {boneName=None}}, {targetBone {boneName=pelvis}}]
    constraintWeights [1,1]
   }
   n122 = "Constraint - dyn_pant_leg_l_b_01_Base (TP)"!(ComponentPose=@n121, Alpha=0.2, ConstraintWeights_0=0, ConstraintWeights_1:float=1) -> (Pose) {
    alphaBoolBlend {blendOption=Linear}
    boneToModify {boneName=dyn_pant_leg_l_b_01_Base}
    constraintSetup [{transformType=Translation, targetBone {boneName=None}}, {targetBone {boneName=pelvis}}]
    constraintWeights [1,1]
   }
   Save cached pose 'CRCache_01'(Pose=Component To Local(ComponentPose=@n122)) {cacheName=CRCache_01}
   n123 = "Transform (Modify) Bone - Bone: dyn_pant_leg_r_f_01"!(ComponentPose, Alpha=@n34, Translation:Vector="0,0,0", Rotation="29.993,0.128,0.751", Scale:Vector="1,1,1") -> (Pose) {
    rotationMode=BMM_Additive
    rotationSpace=BCS_BoneSpace
    translationMode=BMM_Additive
    translationSpace=BCS_BoneSpace
    alphaBoolBlend {blendOption=Linear}
    alphaScaleBiasClamp {inRange {max=500}}
    boneToModify {boneName=dyn_pant_leg_r_f_01}
    rotation=<rot 29.9933 0.128 0.751>
   }
   n124 = "Transform (Modify) Bone - Bone: dyn_pant_leg_l_f_01"(ComponentPose=@n123, Alpha=@n42, Translation="0,0,0", Rotation="29.966,0.410,1.514", Scale="1,1,1") {
    rotationMode=BMM_Additive
    rotationSpace=BCS_BoneSpace
    translationMode=BMM_Additive
    translationSpace=BCS_BoneSpace
    alphaBoolBlend {blendOption=Linear}
    alphaScaleBiasClamp {inRange {max=500}}
    boneToModify {boneName=dyn_pant_leg_l_f_01}
    rotation=<rot 29.9661 0.41 1.514>
   }
   n125 = "Transform (Modify) Bone - Bone: dyn_pant_leg_r_mf_01"!(ComponentPose=@n124, Alpha=@n34, Translation:Vector="0,0,0", Rotation="24.996,-0.095,-0.323", Scale:Vector="1,1,1") -> (Pose) {
    rotationMode=BMM_Additive
    rotationSpace=BCS_BoneSpace
    translationMode=BMM_Additive
    translationSpace=BCS_BoneSpace
    alphaBoolBlend {blendOption=Linear}
    alphaScaleBiasClamp {inRange {max=500}}
    boneToModify {boneName=dyn_pant_leg_r_mf_01}
    rotation=<rot 24.996 -0.0949 -0.3231>
   }
   n126 = "Transform (Modify) Bone - Bone: dyn_pant_leg_l_mf_01"(ComponentPose=@n125, Alpha=@n42, Translation="0,0,0", Rotation="24.999,0.053,0.185", Scale="1,1,1") {
    rotationMode=BMM_Additive
    rotationSpace=BCS_BoneSpace
    translationMode=BMM_Additive
    translationSpace=BCS_BoneSpace
    alphaBoolBlend {blendOption=Linear}
    alphaScaleBiasClamp {inRange {max=500}}
    boneToModify {boneName=dyn_pant_leg_l_mf_01}
    rotation=<rot 24.999 0.053 0.185>
   }
   Reroute Node(InputPin:Component Space Pose Link Structure=@n126) -> (OutputPin:Component Space Pose Link Structure)
  ]}
 EventGraph {nodes [
   n1 = Event BlueprintUpdateAnimation()
   > Set FPS Compensation Alpha(FPS Compensation Alpha:double=MapRangeClamped(Value=@n1.DeltaTimeX, InRangeA=0.02, InRangeB=0.03)) -> (Output_Get:double)
   > n2 = Sequence!(execute) -> (then_0, then_1, then_2:Exec, then_3:Exec)
   n3 = float / float(A=Get PawnSpeed2D(), B=205) {is=title K2Node_CallFunction}
   Set speedINV(execute=@n2.then_1, speedINV:double="Absolute (Float)"(A=float - float(B=@n3) {is=title K2Node_CallFunction})) -> (Output_Get:double)
   n4 = Get Skeletal Mesh() -> (Skeletal Mesh:Skeletal Mesh Component Object Reference)
   Set LeftLegForwardBendCompensationAlpha(
    execute=@n2.then_3
    LeftLegForwardBendCompensationAlpha:double=GetAlphaFromReferenceTransformDelta(SkeletalMeshComponent=@n4, BoneName=thigh_l, BaseName=pelvis, InputRange=(X=-10,Y=-103), TwistAxis="1,0,0")
   ) -> (Output_Get:double)
   > Set RightLegForwardBendCompensationAlpha(RightLegForwardBendCompensationAlpha:double=GetAlphaFromReferenceTransformDelta(SkeletalMeshComponent=@n4, BoneName=thigh_r, BaseName=pelvis, InputRange=(X=-10,Y=-103), TwistAxis="1,0,0")) -> (Output_Get:double)
   > Set bIsNotMoving(bIsNotMoving:bool=NOT Boolean(A=Get bIsMoving2D())) -> (Output_Get:bool)
   n5 = GetOwnerSkeletalMeshComponent()
   Set calfSwingDelta_r(
    execute=@n2.then_0
    calfSwingDelta_r:double=MapRangeClamped(Value=GetTwistAndSwingAngleOfDeltaRotationFromRefPose(self=@n5, BoneName=calf_r).OutSwingAngle, InRangeA=50, InRangeB=120)
   ) -> (Output_Get:double)
   > Set legSwingDelta_l(legSwingDelta_l:double=MapRangeClamped(Value=GetTwistAndSwingAngleOfDeltaRotationFromRefPose(self=@n5, BoneName=thigh_l).OutSwingAngle, InRangeA=24, InRangeB=101)) -> (Output_Get:double)
   n6 = Sequence!(execute=@n2.then_2) -> (then_0, then_1, then_2, then_3, then_4:Exec, then_5:Exec, then_6:Exec)
   Set Crouching(execute=@n6.then_1, Crouching:double="To Float (Boolean)"(InBool=Get bIsCrouching())) -> (Output_Get:double)
   n7 = Branch(
    execute=@n6.then_4
    Condition=OR Boolean!(A=Get bIsSkyDiving(), B=Get bIsSkydiveDiveMode(), C:bool=Get bIsSkydiveFloating(), D:bool=Get bIsParachuteOpened()) -> (ReturnValue)
   )
   > Set SkydiveAlpha(SkydiveAlpha:double=0.5) -> (Output_Get:double)
   Set SkydiveAlpha(execute=@n7.else, SkydiveAlpha=1)
   Set KneeSliding(execute=@n6.then_3, KneeSliding:double="To Float (Boolean)"(InBool=Get bIsKneeSliding())) -> (Output_Get:double)
   n8 = NOT Boolean(A=Get bIsCrouching())
   n9 = Get bIsMoving2D()
   Set isMoving(execute=@n6.then_2, isMoving:double="To Float (Boolean)"(InBool=OR Boolean!(A=@n9, B=@n8) -> (ReturnValue))) -> (Output_Get:double)
   n10 = Get bIsMoving2D()
   n11 = Get bIsCrouching()
   Set CrouchMove(execute=@n6.then_5, CrouchMove:double="To Float (Boolean)"(InBool=AND Boolean(A=@n10, B=@n11))) -> (Output_Get:double)
   n12 = Get bIsSkydiveDiveMode()
   n13 = Get bIsSkydiveFloating()
   Set inAir(execute=@n6.then_6, inAir:double="To Float (Boolean)"(InBool=OR Boolean(A=@n12, B=@n13))) -> (Output_Get:double)
   n14 = Get bIsJumping()
   n15 = Get bIsFalling()
   Set Jumping(execute=@n6.then_0, Jumping:double="To Float (Boolean)"(InBool=OR Boolean(A=@n14, B=@n15))) -> (Output_Get:double)
   Event BlueprintInitializeAnimation()
   > Set Skeletal Mesh(Skeletal Mesh:Skeletal Mesh Component Object Reference=GetOwnerSkeletalMeshComponent()) -> (Output_Get:Skeletal Mesh Component Object Reference)
   TryGetPawnOwner()
  ]}
]
variables [
 bIsNotMoving {type=boolean}
 calfSwingDelta_r {type=number}
 Crouching {type=number}
 CrouchMove {type=number}
 FPS Compensation Alpha {type=number}
 inAir {type=number}
 isMoving {type=number}
 Jumping {type=number}
 KneeSliding {type=number}
 LeftLegForwardBendCompensationAlpha {type=number}
 legSwingDelta_l {type=number}
 RightLegForwardBendCompensationAlpha {type=number}
 Skeletal Mesh {type=SkeletalMeshComponent}
 SkydiveAlpha {type=number}
 speedINV {type=number}
]
